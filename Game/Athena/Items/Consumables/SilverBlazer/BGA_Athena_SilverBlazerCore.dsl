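; SilverBlazer core building actor: a deployed device that spawns and carries a bubble safe zone actor.
; On authority it attaches the bubble to its mesh; core death ends the bubble, and bubble death destroys the core.
; A sphere trace along projectile velocity, or downward, finds vehicles to attach to by closest bone, which the bubble then ignores.
; An EMP tag kills the core, and vehicle destruction detaches it.
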(event Custom|HandleEMP
  ;[1 1 1] Handle EMP
  (Gameplay|Die (Variables|Default|GetTag_EMPApplied)))

(event EventBeginPlay
  ;[1 1 1] BeginPlay
  (|CallParentFunction)
  (Utilities|FlowControl|Sequence
    (:then_0
      (EventDispatchers|BindEventtoOnDied (EventDispatchers|CreateEvent :!event Custom|Destroyed))
      (Networking|SwitchHasAuthority
        (:Authority
          (Utilities|FlowControl|DoOnce false
            (:Completed
              (bind _spawnedbubble (CallFunction|SpawnBubbleSafeZone))
              (bind _output_get (Variables|Default|SetSpawnedBGA _spawnedbubble))
              (Transformation|AttachComponentToComponent (Variables|/Script/Engine.Actor|GetRootComponent _output_get) (Variables|Default|GetStaticMesh) "None" "KeepWorld" "KeepWorld" "KeepWorld" true)
              (EventDispatchers|BindEventtoOnDied (Variables|Default|GetSpawnedBGA) (EventDispatchers|CreateEvent :!event Custom|Die)))))))
    (:then_1
      (CallFunction|CheckforVehicleAttach))
    (:then_2
      (Ability|Tasks|WaitGameplayTagAddToActor 0 (Variables|Default|GetTag_EMPApplied)
        (:Added
          (CallFunction|HandleEMP))))))

(event Custom|Destroyed
  (params (DamagedActor :type Actor%20Object%20Reference) (Damage :type Float%20%28single-precision%29) (InstigatedBy :type Controller%20Object%20Reference) (DamageCauser :type Actor%20Object%20Reference) (HitLocation :type Vector) (FHitComponent :type Primitive%20Component%20Object%20Reference) (BoneName :type Name) (Momentum :type Vector))
  (Niagara|SpawnSystematLocation (Variables|Default|GetDestroyedFX) (Transformation|GetActorLocation))
  (Rendering|SetActorHiddenInGame true)
  (Collision|SetActorEnableCollision false)
  (Class|BBGAAthenaSilverBlazerV2|End (Variables|Default|GetSpawnedBGA)))

(event Custom|Die
  (params (DamagedActor :type Actor%20Object%20Reference) (Damage :type Float%20%28single-precision%29) (InstigatedBy :type Controller%20Object%20Reference) (DamageCauser :type Actor%20Object%20Reference) (HitLocation :type Vector) (FHitComponent :type Primitive%20Component%20Object%20Reference) (BoneName :type Name) (Momentum :type Vector))
  (Actor|DestroyActor))

(event Custom|CheckForVehicleAttach
  (bind _spawnedbga (Variables|Default|GetSpawnedBGA))
  (bind (_didattach _attachedvehicle) (CallFunction|TryAttachtoVehicle))
  (if _didattach
    (Game|BindEventtoOnDestroyed _attachedvehicle (delegate Custom|OnAttachedVehicleDestroyed))
    (Utilities|IsValid _spawnedbga
      (:Is%20Valid
        (Class|BBGAAthenaSilverBlazerV2|SetIgnoreActor _spawnedbga _attachedvehicle)))))

(event Custom|OnAttachedVehicleDestroyed
  (params (DestroyedActor :type Actor%20Object%20Reference))
  (Transformation|DetachFromActor "KeepWorld" "KeepWorld" "KeepWorld"))

(fn FindAttachableVehicle
  (bind (_outhits _returnvalue) (Collision|MultiSphereTraceForObjects (Transformation|GetActorLocation) (+ (Transformation|GetActorLocation) (* (Math|Vector|SelectVector (Variables|/Script/Engine.MovementComponent|GetVelocity (Variables|Default|GetFortProjectileMovement)) "0, 0,-1.000000" (> (Math|Vector|VectorLength (Variables|/Script/Engine.MovementComponent|GetVelocity (Variables|Default|GetFortProjectileMovement))) 0)) (Variables|Default|GetVehicleCheckDistance))) 5e+01 (Utilities|Array|MakeArray "ObjectTypeQuery5")))
  (if _returnvalue
    (for _array_element _outhits
      (if (Utilities|DoesObjectImplementInterface (.HitActor (Collision|BreakHitResult _array_element)) "/Script/FortniteGame.FortVehicleInterface")
        (return true (.HitActor (Collision|BreakHitResult _array_element)))
        (else
          (return false 0))))
    (else
      _)))

(fn OnRep_SpawnedBGA
  (Class|BBGAAthenaSilverBlazerV2|SetFXPos (Variables|Default|GetSpawnedBGA) (Transformation|GetActorLocation)))

(fn SpawnBubbleSafeZone
  (bind _returnvalue_location (Math|Transform|BreakTransform (Transformation|GetActorTransform)))
  ;[0 0.07676 1] SpawnBubbleSafeZone
  ; Spawn the bubble area BGA
  (bind _returnvalue (Game|SpawnBuildingActorfromClass "/Game/Athena/Items/Consumables/SilverBlazer/B_BGA_Athena_SilverBlazer_V2.B_BGA_Athena_SilverBlazer_V2_C" (Math|Transform|MakeTransform _returnvalue_location) 0 false 0.0 (Variables|Missions|GetAssociatedMissionParam) (Game|GetInstigator)))
  (return _returnvalue))

(fn TryAttachToVehicle
  (bind (_canattach _hitactor) (CallFunction|FindAttachableVehicle))
  (if _canattach
    (bind _asfort_vehicle_interface (Utilities|Casting|CastToFortVehicleInterface _hitactor))
    (bind _returnvalue (Transformation|AttachComponentToComponent (Variables|Default|GetStaticMesh) (Class|FortVehicleInterface|GetSkeletalMeshComponent _asfort_vehicle_interface) (.ReturnValue (Components|SkinnedMesh|FindClosestBone (Class|FortVehicleInterface|GetSkeletalMeshComponent _asfort_vehicle_interface) (Transformation|GetActorLocation) 0.0 false)) "KeepWorld" "KeepWorld" "KeepWorld" true))
    (Collision|SetCollisionEnabled (Variables|Default|GetStaticMesh) "QueryOnly")
    (return _returnvalue _hitactor)
    (else
      (return false 0))))

(fn ConstructionScript
  (|CallParentFunction))
