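; Pocket shield bubble gameplay actor (SilverBlazer) that deploys a timed protective dome, ending early on EMP.
; A timeline scales the SafeZone through a native update, then registers an AI occluder and enables the player-overlap sphere.
; Damage and hits drive impact FX and material ripples scaled by damage or weapon type, while entering or exiting triggers passthrough FX and quest updates.
; A warning pulse precedes expiry, then a dissolve timeline destroys it.
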(event Custom|HandleEMP
  ;[1 1 1] Handle EMP
  (CallFunction|End))

(event Custom|PlaySpawnSound
  ;[0 0.1106 1] RPC to play spawn sound - keeps things in order for replay.
  (Audio|PlaySoundatLocation (Variables|Sounds|GetSoundActivate) (Transformation|GetActorLocation)))

(event Custom|PlayWarningEffectBeforeShieldGoAway
  (bind _safezone (Variables|Default|GetSafeZone))
  ;[1 0.5308 0.5507] WarningFX Before Shield Go away
  (Rendering|Material|SetVectorParameterValueonMaterials (Variables|Default|GetSafeZone) "ColorCenter" "0.027000,0.008388,0.338542")
  (bind (_direction _warningemissive) (Timeline|Timeline_Warning PlayFromStart)
    (:Update
      (Rendering|Material|SetScalarParameterValueonMaterials _safezone "FresnelBrightness" (Math|Float|Lerp 1.0 12.0 _warningemissive))
      (Rendering|Material|SetScalarParameterValueonMaterials _safezone "IntBrightness" (Math|Float|Lerp 1.5 12.0 _warningemissive)))
    (:Audio%20Pulse
      (Audio|SpawnSoundAttached (Variables|Sounds|GetSoundPulse) (Variables|Default|GetRoot)))))

(event EventBeginPlay
  ;[1 1 1] EVENT -> BeginPlay
  (Variables|Default|SetIsEnding false)
  (Utilities|Time|SetTimerbyEvent (EventDispatchers|CreateEvent :!event Custom|End) (ScalableFloat|GetValueatLevel (Variables|SilverBronco|GetLifespanTime)))
  (bind _returnvalue (Effects|SpawnFXSystemAttached :FXSystemAsset (Variables|FX|GetFX_Activate) :AttachToComponent (Variables|Default|GetRoot) :AttachPointName "None" :Location "0.000000,0.000000,20.000000" :Scale "1.000000,1.000000,1.000000" :LocationType "SnapToTarget" :bAutoDestroy true :PoolingMethod "None" :bAutoActivate true))
  (Variables|Default|SetFX_SpawnedDeployComponent_Ref _returnvalue)
  (Ability|Tasks|WaitGameplayTagAddToActor 0 (Variables|Default|GetTag_EMPApplied)
    (:then
      ;[1 1 1] Spawn Sound, FX and play Animations
      (Utilities|FlowControl|DoOnce false
        (:Completed
          (Quest|SendComplexCustomStatEvent (Quests|GetQuestManager (Fortnite|GetFortPlayerControllerfromActor (Game|GetInstigator)) "Invalid") 0 "(GameplayTags=((TagName=\"Athena.Quests.ShieldBubble\")))" "(GameplayTags=)" 1)
          (CallFunction|EventScaleSafeZone)
          (Networking|SwitchHasAuthority
            (:Authority
              (CallFunction|PlaySpawnSound))
            (:Remote
              _))
          (Utilities|FlowControl|Delay (- (ScalableFloat|GetValueatLevel (Variables|SilverBronco|GetLifespanTime)) 3.0))
          (CallFunction|PlayWarningEffectBeforeShieldGoAway))))
    (:Added
      (CallFunction|HandleEMP))))

(event Custom|End
  (bind _returnvalue (Class|AthenaAIServiceZoneManager|GetAthenaAIServiceZoneManager true))
  ;[1 1 1] KillSelf
  (if (not (Variables|Default|GetIsEnding))
    (Variables|Default|SetIsEnding true)
    (Quests|BeginorEndQuestUpdate false)
    (CallFunction|DeathCosmetics)
    (Audio|PlaySoundatLocation (Variables|Sounds|GetSoundDeactivate) (Transformation|GetActorLocation))
    (Utilities|IsValid _returnvalue
      (:Is%20Valid
        (Class|AthenaAIServiceZoneManager|UnregisterOccluder _returnvalue self)))))

(event Custom|DeathCosmetics
  (bind _safezone (Variables|Default|GetSafeZone))
  (bind _safezone_2 (Variables|Default|GetSafeZone))
  (bind _fx_spawneddeploycomponent_ref (Variables|Default|GetFX_SpawnedDeployComponent_Ref))
  ;[1 1 1] Death Cosmetics
  (Utilities|FlowControl|RetriggerableDelay 0.1)
  (Rendering|Material|SetScalarParameterValueonMaterials _safezone "FresnelBrightness" 1.0)
  (Rendering|Material|SetScalarParameterValueonMaterials _safezone "IntBrightness" 1.5)
  (bind (_direction _fadeout _shrinkwpo _opacity) (Timeline|TimeLine_Death PlayFromStart)
    (:Update
      (Rendering|Material|SetScalarParameterValueonMaterials _safezone_2 "DissolveAmount" _fadeout)
      (Rendering|Material|SetScalarParameterValueonMaterials _safezone_2 "Shrink" _shrinkwpo)
      (Rendering|Material|SetScalarParameterValueonMaterials _safezone_2 "OverallOpacity" _opacity))
    (:Finished
      (Rendering|Material|SetVectorParameterValueonMaterials (Variables|Default|GetSafeZone) "ColorCenter" "0.031414,0.163287,0.338542")
      (Rendering|SetActorHiddenInGame self true)
      (Actor|DestroyActor))
    (:TriggerDeathFX
      (Components|DestroyComponent _fx_spawneddeploycomponent_ref)
      (Components|Activation|Deactivate _fx_spawneddeploycomponent_ref)
      (Effects|Components|ParticleSystem|SpawnFXAtLocation (Variables|FX|GetFX_Death) (Transformation|GetActorLocation))
      (Components|DestroyComponent (Variables|Default|GetAudioGameplayVolume)))))

(event Custom|EventScaleSafeZone
  (bind (_origin _boxextent _sphereradius) (Collision|GetComponentBounds (Variables|Default|GetSafeZone)))
  (bind _returnvalue (Class|AthenaAIServiceZoneManager|GetAthenaAIServiceZoneManager true))
  (bind _fx_spawneddeploycomponent_ref (Variables|Default|GetFX_SpawnedDeployComponent_Ref))
  ;[0.4503 0.7522 1] Scale Safe Zone & Deploy FX
  (Components|Timeline|SetPlayRate (Variables|B_BGA_Athena_SilverBlazer_Pocket|GetScaleSafeZone) (/ 1.0 (ScalableFloat|GetValueatLevel (Variables|SilverBronco|GetScaleDuration))))
  (Utilities|Struct|SetmembersinSilverBlazerScaleSafeZoneUpdateData (Variables|Default|GetScaleSafeZoneUpdateData) "Deploy" "FresnelBrightness" (Variables|Default|GetSafeZone) (ScalableFloat|GetValueatLevel (Variables|SilverBronco|GetMaxScale)) 5e+01 1.0 5.0)
  (bind _audiogameplayvolume (Variables|Default|GetAudioGameplayVolume)
    (:then
      (bind _asagv_silver_blazer (Utilities|Casting|CastToAGV_SilverBlazer (Variables|/Script/Engine.ChildActorComponent|GetChildActor _audiogameplayvolume))
        (:then
          (Utilities|Struct|SetmembersinSilverBlazerScaleSafeZoneUpdateData (Variables|Default|GetScaleSafeZoneUpdateData) (Variables|/Game/Athena/Items/Consumables/SilverBlazer/AGV_SilverBlazer.AGV_SilverBlazer_C|GetAudioShapeSphere _asagv_silver_blazer)))
        (:CastFailed
          _))
      (bind (_direction _scalesafezone _deployscanin _deploydoneemis) (Timeline|ScaleSafeZone Play)
        (:Update
          (bind _structout (Utilities|Struct|SetmembersinSilverBlazerScaleSafeZoneUpdateData (Variables|Default|GetScaleSafeZoneUpdateData) _scalesafezone _deployscanin _deploydoneemis))
          (CallFunction|NativeScaleSafeZoneUpdate _structout)
          (Utilities|IsValid _fx_spawneddeploycomponent_ref
            (:Is%20Valid
              (Effects|Components|ParticleSystem|SetFloatParameter _fx_spawneddeploycomponent_ref "Scalar" (/ (* _scalesafezone (ScalableFloat|GetValueatLevel (Variables|SilverBronco|GetMaxScale))) 1e+01)))))
        (:Finished
          (Utilities|FlowControl|Sequence
            (:then_0
              (Utilities|IsValid _returnvalue
                (:Is%20Valid
                  (Class|AthenaAIServiceZoneManager|RegisterOccluder _returnvalue self _sphereradius))))
            (:then_1
              (Collision|SetCollisionEnabled (Variables|Default|GetCollisionSphere) "QueryOnly") ; wait until the bubble is fully deployed to check if there are players within it
            )))))
    (:else
      _)))

(event Collision|EventHit
  (params (MyComp :type Primitive%20Component%20Object%20Reference) (Other :type Actor%20Object%20Reference) (OtherComp :type Primitive%20Component%20Object%20Reference) (bSelfMoved :type Boolean) (HitLocation :type Vector :default "0, 0, 0") (HitNormal :type Vector :default "0, 0, 0") (NormalImpulse :type Vector :default "0, 0, 0") (Hit :type Hit%20Result%20Structure%20%28by%20ref%29))
  (bind (_asfort_player_pawn _bsuccess) (Utilities|Casting|CastToFortPlayerPawn Other))
  (bind (_asfort_vehicle_interface _bsuccess_2) (Utilities|Casting|CastToFortVehicleInterface Other))
  ;[0.7125 0.5853 1] EVENT -> On Hit
  (Utilities|FlowControl|Sequence
    (:then_0
      (if (Math|Boolean|ANDBoolean (Utilities|Equal(Object) (Variables|Default|GetSafeZone) MyComp) (not _bsuccess) (not _bsuccess_2))
        (Audio|PlaySoundatLocation (Variables|Sounds|GetSoundHitSafeZone) (Transformation|GetActorLocation))
        (Effects|Components|ParticleSystem|SpawnFXAtLocation (Variables|FX|GetFX_HitSafeZone) HitLocation (Math|Conversions|RotationFromXVector HitNormal))))
    (:then_1
      (bind _asfort_weapon (Utilities|Casting|CastToFortWeapon (Actor|GetOwner Other)))
      (if (or (GameplayTags|HasTag (GameplayTags|GetOwnedGameplayTags (Data|GetWeaponData _asfort_weapon) :!class /Script/GameplayTags.GameplayTagAssetInterface) "(TagName=\"Weapon.Ranged.Heavy.Rocket_Launcher\")" true) (GameplayTags|HasTag (GameplayTags|GetOwnedGameplayTags (Data|GetWeaponData _asfort_weapon) :!class /Script/GameplayTags.GameplayTagAssetInterface) "(TagName=\"Ability.Ostrich.SecondaryFire\")" true))
        (CallFunction|HitbyRocket HitLocation)
        (else
          (Development|PrintString "Is Not hit by Rocket"))))))

(event EventOnDamageServer
  (params (Damage :type Float%20%28single-precision%29) (DamageTags :type Gameplay%20Tag%20Container%20Structure%20%28by%20ref%29) (Momentum :type Vector :default "0, 0, 0") (HitInfo :type Hit%20Result%20Structure%20%28by%20ref%29) (InstigatedBy :type Controller%20Object%20Reference) (DamageCauser :type Actor%20Object%20Reference) (EffectContext :type Gameplay%20Effect%20Context%20Handle%20Structure))
  ;[1 1 1] Block Projectiles Quest
  (Quest|SendComplexCustomStatEvent (Quests|GetQuestManager (Fortnite|GetFortPlayerControllerfromActor (Game|GetInstigator)) "Athena") 0 "(GameplayTags=((TagName=\"Athena.Quests.ShieldBubbleBlockProj\")))" "(GameplayTags=)" 1))

(event OnComponentBeginOverlap(CollisionSphere)
  (params (OverlappedComponent :type Primitive%20Component%20Object%20Reference) (OtherActor :type Actor%20Object%20Reference) (OtherComp :type Primitive%20Component%20Object%20Reference) (OtherBodyIndex :type Integer) (bFromSweep :type Boolean) (SweepResult :type Hit%20Result%20Structure%20%28by%20ref%29))
  ;[0.5548 0.4757 1] Play FX while running in and out
  (Networking|SwitchHasAuthority
    (:Authority
      (Quests|QuestUpdateonEnterBubble OtherActor)
      (Utilities|DoOncePerTimeFrame 0.4
        (:Then
          (CallFunction|RunIn OtherActor))))))

(event Custom|RunIn
  (params (Actor :type Actor%20Object%20Reference))
  (bind _runinlenseffectinterface (Variables|Default|GetRunInLensEffectInterface))
  ;[0.5548 0.4757 1] Play FX while running in and out
  (Utilities|FlowControl|Delay 0.0)
  (bind _asactor (Utilities|Casting|CastToActor (Variables|Default|GetRunOutLensEffectInterface))
    (:then
      (Actor|DestroyActor _asactor))
    (:CastFailed
      _))
  (bind _asplayer_pawn_athena (Utilities|Casting|CastToPlayerPawn_Athena Actor))
  (Utilities|FlowControl|Sequence
    (:then_0
      (bind _returnvalue (Feedback|SpawnGenericCameraLensEffect _asplayer_pawn_athena (Variables|FX|GetPassthroughLensEffectEmitterClass)))
      (Variables|Default|SetRunInLensEffectInterface _returnvalue)
      (Utilities|FlowControl|Delay 0.5)
      (bind _asactor_2 (Utilities|Casting|CastToActor _runinlenseffectinterface))
      (Components|Activation|Deactivate (CameraLensEffect|GetPrimaryParticleComponent _runinlenseffectinterface))
      (Actor|SetLifeSpan _asactor_2 0.5))
    (:then_1
      (if (not (Variables|Default|GetIsEnding))
        (bind _output_get (Variables|Default|SetPlayerGoesInLoc (Transformation|GetActorLocation _asplayer_pawn_athena)))
        (Audio|SpawnSoundatLocation (Variables|Sounds|GetSoundEnterBubble) _output_get)
        (bind _returnvalue_2 (Niagara|SpawnSystematLocation (Variables|Default|GetPassthroughFX) (Variables|Default|GetPlayerGoesInLoc) (Math|Rotator|FindLookatRotation (Variables|Default|GetPlayerGoesInLoc) (Transformation|GetActorLocation))))
        (Niagara|SetNiagaraVariable(Int32) _returnvalue_2 "SpawnBurst" 1)
        (Niagara|SetNiagaraVariable(Float) _returnvalue_2 "FlareLIfetime" 0.6)
        (Niagara|SetNiagaraVariable(Float) _returnvalue_2 "SpawnRate" 8e+01)))))

(event OnComponentEndOverlap(CollisionSphere)
  (params (OverlappedComponent :type Primitive%20Component%20Object%20Reference) (OtherActor :type Actor%20Object%20Reference) (OtherComp :type Primitive%20Component%20Object%20Reference) (OtherBodyIndex :type Integer))
  ;[0.5548 0.4757 1] Play FX while running in and out
  (Networking|SwitchHasAuthority
    (:Authority
      (Quests|QuestUpdateonExitBubble OtherActor)
      (Utilities|DoOncePerTimeFrame 0.4
        (:Then
          (CallFunction|RunOut OtherActor))))))

(event Custom|RunOut
  (params (Actor :type Actor%20Object%20Reference))
  (bind _runoutlenseffectinterface (Variables|Default|GetRunOutLensEffectInterface))
  ;[0.5548 0.4757 1] Play FX while running in and out
  (Utilities|FlowControl|Delay 0.0)
  (bind _asactor (Utilities|Casting|CastToActor (Variables|Default|GetRunInLensEffectInterface))
    (:then
      (Actor|DestroyActor _asactor))
    (:CastFailed
      _))
  (bind _asplayer_pawn_athena (Utilities|Casting|CastToPlayerPawn_Athena Actor))
  (Utilities|FlowControl|Sequence
    (:then_0
      (bind _returnvalue (Feedback|SpawnGenericCameraLensEffect _asplayer_pawn_athena (Variables|FX|GetPassthroughLensEffectEmitterClass)))
      (Variables|Default|SetRunOutLensEffectInterface _returnvalue)
      (Utilities|FlowControl|Delay 0.5)
      (bind _asactor_2 (Utilities|Casting|CastToActor _runoutlenseffectinterface))
      (Components|Activation|Deactivate (CameraLensEffect|GetPrimaryParticleComponent _runoutlenseffectinterface))
      (Actor|SetLifeSpan _asactor_2 0.5))
    (:then_1
      (if (not (Variables|Default|GetIsEnding))
        (bind _output_get (Variables|Default|SetPlayerGoesOutLoc (Transformation|GetActorLocation _asplayer_pawn_athena)))
        (Audio|SpawnSoundatLocation (Variables|Sounds|GetSoundExitBubble) _output_get)
        (bind _returnvalue_2 (Niagara|SpawnSystematLocation (Variables|Default|GetPassthroughFX) (Variables|Default|GetPlayerGoesOutLoc) (Math|Rotator|FindLookatRotation (Transformation|GetActorLocation) (Variables|Default|GetPlayerGoesOutLoc))))
        (Niagara|SetNiagaraVariable(Int32) _returnvalue_2 "SpawnBurst" 0)
        (Niagara|SetNiagaraVariable(Float) _returnvalue_2 "FlareLIfetime" 0.3)
        (Niagara|SetNiagaraVariable(Float) _returnvalue_2 "SpawnRate" 0.0)))))

(event EventOnDamagePlayEffects
  (params (Damage :type Float%20%28single-precision%29) (DamageTags :type Gameplay%20Tag%20Container%20Structure%20%28by%20ref%29) (Momentum :type Vector :default "0, 0, 0") (HitInfo :type Hit%20Result%20Structure%20%28by%20ref%29) (InstigatedBy :type Fort%20Pawn%20Object%20Reference) (DamageCauser :type Actor%20Object%20Reference) (EffectContext :type Gameplay%20Effect%20Context%20Handle%20Structure))
  (bind _damage (Variables|Default|GetDamage))
  (bind _returnvalue (Math|Float|NormalizetoRange _damage 5.0 45.0))
  (bind (_bblockinghit _binitialoverlap _time _distance _location _impactpoint _normal) (Collision|BreakHitResult HitInfo))
  ;[0.6147 0.4092 1] EVENT -> On Damage
  (Utilities|FlowControl|Sequence
    (:then_0
      (Variables|Default|SetDamage Damage)
      (Audio|PlaySoundatLocation (Variables|Sounds|GetSoundHitSafeZone) _location)
      (bind _returnvalue_2 (Effects|Components|ParticleSystem|SpawnFXAtLocation (Variables|FX|GetFX_HitSafeZone) _location (Math|Conversions|RotationFromXVector _normal) "1.000000,1.000000,1.000000" true "None" false))
      (Variables|Default|SetFX_SafezoneImpact _returnvalue_2)
      (Variables|Default|SetHitWaveLocation _impactpoint))
    (:then_1
      (bind _asfort_weapon (Utilities|Casting|CastToFortWeapon DamageCauser))
      (if (GameplayTags|HasTag (GameplayTags|GetOwnedGameplayTags (Data|GetWeaponData _asfort_weapon) :!class /Script/GameplayTags.GameplayTagAssetInterface) "(TagName=\"Weapon.Ranged.Shotgun\")" true)
        (Development|PrintString "Is hit by Shotgun")
        (CallFunction|HitbyShotgun _location)
        (CallFunction|SetRippleFXParam _impactpoint 3e+02 1e+03 0.6 1.0)
        (else
          (bind _fx_safezoneimpact (Variables|Default|GetFX_SafezoneImpact)
            (:then
              (Effects|Components|ParticleSystem|SetFloatParameter _fx_safezoneimpact "GridSize" _returnvalue)
              (Effects|Components|ParticleSystem|SetFloatParameter _fx_safezoneimpact "GridRange" (Math|Float|Lerp 5.0 55.0 _returnvalue))
              (Components|Activation|Activate _fx_safezoneimpact false)
              (CallFunction|SetRippleFXParam (Variables|Default|GetHitWaveLocation) (Math|Float|Lerp 3e+01 3e+02 _returnvalue) (Math|Float|Lerp 1e+02 1e+03 _returnvalue) (Math|Float|Lerp 0.03 0.45 _returnvalue) 1.0))
            (:else)))))))

(event Custom|HitByShotgun
  (params (ImpactLoc :type Vector))
  ;[0.6147 0.4092 1] EVENT -> On Damage
  (bind _fx_safezoneimpact (Variables|Default|GetFX_SafezoneImpact)
    (:then
      (Effects|Components|ParticleSystem|SetFloatParameter _fx_safezoneimpact "ImpactHoleRange" 45.0)
      (Effects|Components|ParticleSystem|SetIntParameter _fx_safezoneimpact "ImpactHoleAmount" 8)
      (Components|Activation|Activate _fx_safezoneimpact false))
    (:else
      (Development|PrintString "NotValid"))))

(event Custom|HitByRocket
  (params (ImpactLoc :type Vector))
  ;[0.6147 0.4092 1] EVENT -> On Damage
  (CallFunction|SetRippleFXParam ImpactLoc 5e+02 1.3e+03 0.95 1.0))

(event OnComponentBeginOverlap(SafeZone)
  (params (OverlappedComponent :type Primitive%20Component%20Object%20Reference) (OtherActor :type Actor%20Object%20Reference) (OtherComp :type Primitive%20Component%20Object%20Reference) (OtherBodyIndex :type Integer) (bFromSweep :type Boolean) (SweepResult :type Hit%20Result%20Structure%20%28by%20ref%29)))

(event EventEndPlay
  (params (EndPlayReason :type EEndPlayReason%20Enum :default "Destroyed"))
  ;[1 0.4803 0.03552] Clear references to ignored actors when this shield bubble is destroyed.
  (|CallParentFunction EndPlayReason)
  (for _array_element (Variables|Default|GetActorsToIgnoreWhileMoving)
    (Utilities|IsValid _array_element
      (:Is%20Valid
        (Game|UnbindEventfromOnDestroyed _array_element (EventDispatchers|CreateEvent :!event Custom|OnMoveIgnoreActorDestroyed))
        (Collision|IgnoreActorwhenMoving (Variables|Default|GetCollisionSphere) _array_element false))))
  (Utilities|Array|Clear (Variables|Default|GetActorsToIgnoreWhileMoving)))

(event Custom|OnMoveIgnoreActorDestroyed
  (params (DestroyedActor :type Actor%20Object%20Reference))
  ;[1 0.4803 0.03552] Handle an ignored actor being destroyed. Remove it from the MoveIgnoredActors list in PrimitiveComponent to avoid garbage references.
  (Utilities|IsValid DestroyedActor
    (:Is%20Valid
      (Collision|IgnoreActorwhenMoving (Variables|Default|GetCollisionSphere) DestroyedActor false)
      (Utilities|Array|RemoveItem (Variables|Default|GetActorsToIgnoreWhileMoving) DestroyedActor))))

(fn BeginOrEndQuestUpdate
  (params (BubbleIsStarting :type Boolean))
  (for _array_element (Collision|GetOverlappingActors (Variables|Default|GetCollisionSphere) "/Script/FortniteGame.FortPawn" :!class /Script/Engine.PrimitiveComponent)
    (if BubbleIsStarting
      (Quests|QuestUpdateonEnterBubble _array_element)
      (else
        (Quests|QuestUpdateonExitBubble _array_element)))))

(fn GetAffiliationsToShowFor
  (return (LocalVariables|Getseer)))

(fn GetResistenceInformation
  (bind (_origin _boxextent _sphereradius) (Collision|GetComponentBounds (Variables|Default|GetSafeZone)))
  (return (Transformation|GetActorLocation) _sphereradius 1.5e+02))

(fn QuestUpdateOnEnterBubble
  (params (OverlapActor :type Actor%20Object%20Reference))
  (bind _asfort_player_pawn_athena (Utilities|Casting|CastToFortPlayerPawnAthena OverlapActor))
  (Utilities|IsValid (Ability|GetFortAbilitySystemComponent _asfort_player_pawn_athena)
    (:Is%20Valid
      (GameplayEffects|FortApplyGameplayEffectToTarget :self (Ability|GetFortAbilitySystemComponent _asfort_player_pawn_athena) :GameplayEffectClass (Variables|Quests|GetQuestGE_WithinBubbleTag) :TargetOfEffect (Ability|GetFortAbilitySystemComponent _asfort_player_pawn_athena) :OptionalSourceObject self :OptionalEffectCauser self :Level 0.0)
      (Quest|SendComplexCustomStatEvent (Quests|GetQuestManager (Fortnite|GetFortPlayerControllerfromActor _asfort_player_pawn_athena) "Invalid") (Quests|GetQuestManager (Fortnite|GetFortPlayerControllerfromActor _asfort_player_pawn_athena) "Invalid") (Variables|Quests|GetQuestTag_EnteredComplexCustom) (GameplayTags|GetOwnedGameplayTags _asfort_player_pawn_athena :!class /Script/GameplayTags.GameplayTagAssetInterface) 1) ; DEPRICATED (?) Quest event
      (Ability|SendGameplayEventtoActor _asfort_player_pawn_athena "(TagName=\"Athena.Quest.ShieldBubble.EventEntered\")" (Utilities|Struct|MakeGameplayEventData "(TagName=\"\")" _asfort_player_pawn_athena self 0 0 "()" (GameplayTags|GetOwnedGameplayTags _asfort_player_pawn_athena :!class /Script/GameplayTags.GameplayTagAssetInterface) (Tags|GetGameplayTags self))) ; Send Quest Event On Exit
    )))

(fn QuestUpdateOnExitBubble
  (params (OverlapActor :type Actor%20Object%20Reference))
  (bind _asfort_player_pawn_athena (Utilities|Casting|CastToFortPlayerPawnAthena OverlapActor))
  (Utilities|IsValid (Ability|GetFortAbilitySystemComponent _asfort_player_pawn_athena)
    (:Is%20Valid
      (GameplayEffects|RemoveActiveGameplayEffectbySourceEffect (Ability|GetFortAbilitySystemComponent _asfort_player_pawn_athena) (Variables|Quests|GetQuestGE_WithinBubbleTag) (Ability|GetFortAbilitySystemComponent _asfort_player_pawn_athena) 1)
      (Quest|SendComplexCustomStatEvent (Quests|GetQuestManager (Fortnite|GetFortPlayerControllerfromActor _asfort_player_pawn_athena) "Invalid") _asfort_player_pawn_athena (Variables|Default|GetQuestTag_ExitComplexCustom) (GameplayTags|GetOwnedGameplayTags _asfort_player_pawn_athena :!class /Script/GameplayTags.GameplayTagAssetInterface) 1) ; DEPRICATED (?) Quest event
      (Ability|SendGameplayEventtoActor _asfort_player_pawn_athena "(TagName=\"Athena.Quest.ShieldBubble.EventExited\")" (Utilities|Struct|MakeGameplayEventData "(TagName=\"\")" _asfort_player_pawn_athena self 0 0 "()" (GameplayTags|GetOwnedGameplayTags _asfort_player_pawn_athena :!class /Script/GameplayTags.GameplayTagAssetInterface) (Tags|GetGameplayTags self))) ; Send Quest Event On Enter
    )))

(fn ScaleAudioGameplayVolume
  (params (InVec :type Vector))
  (bind _audiogameplayvolume (Variables|Default|GetAudioGameplayVolume)
    (:then
      (bind _asagv_silver_blazer (Utilities|Casting|CastToAGV_SilverBlazer (Variables|/Script/Engine.ChildActorComponent|GetChildActor _audiogameplayvolume)))
      (Control|SetRadius (Variables|/Game/Athena/Items/Consumables/SilverBlazer/AGV_SilverBlazer.AGV_SilverBlazer_C|GetAudioShapeSphere _asagv_silver_blazer) (* (.x InVec) 5e+01) :!class /Script/AudioShapes.AudioShapeSphereComponent))
    (:else)))

(fn Set_FX_Pos
  (params (Pos :type Vector))
  (Transformation|SetWorldLocation (Variables|Default|GetFX_SpawnedDeployComponent_Ref) Pos false true))

(fn Set_RippleFX_Param
  (params (HitWave-Loc :type Vector) (HitWave-SizeMultiplierOuterRing :type Float%20%28double-precision%29) (HitWave-SizeMultiplierInnerRing :type Float%20%28double-precision%29) (HitWave-DurationNew :type Float%20%28double-precision%29) (HitGridBrightness :type Float%20%28double-precision%29))
  (bind _safezone (Variables|Default|GetSafeZone))
  (Rendering|Material|SetVectorParameterValueonMaterials _safezone "HitWave-Loc" HitWave-Loc)
  (Rendering|Material|SetScalarParameterValueonMaterials _safezone "GameTime" (Utilities|Time|GetGameTimeinSeconds))
  (Rendering|Material|SetScalarParameterValueonMaterials _safezone "HitWave-SizeMultiplierOuterRing" HitWave-SizeMultiplierOuterRing)
  (Rendering|Material|SetScalarParameterValueonMaterials _safezone "HitWave-SizeMultiplierInnerRing" HitWave-SizeMultiplierInnerRing)
  (Rendering|Material|SetScalarParameterValueonMaterials _safezone "HitWave-DurationNew" HitWave-DurationNew)
  (Rendering|Material|SetScalarParameterValueonMaterials _safezone "HitEffectLerp" HitGridBrightness))

(fn SetIgnoreActor
  (params (Actor :type Actor%20Object%20Reference))
  (bind _actorstoignorewhilemoving (Variables|Default|GetActorsToIgnoreWhileMoving))
  (Utilities|IsValid Actor
    (:Is%20Valid
      (if (not (Utilities|Array|ContainsItem _actorstoignorewhilemoving Actor))
        (Utilities|Array|Add _actorstoignorewhilemoving Actor)
        (Collision|IgnoreActorwhenMoving (Variables|Default|GetCollisionSphere) Actor true)
        (Game|BindEventtoOnDestroyed Actor (EventDispatchers|CreateEvent :!event OnMoveIgnoreActorDestroyed))))))

(fn ShouldShowSoundIndicator
  (params (PlayerController :type Fort%20Player%20Controller%20Object%20Reference))
  (return (!= PlayerController (Fortnite|GetFortPlayerControllerfromActor (Game|GetInstigator)))))

(fn ConstructionScript
  (|CallParentFunction))
