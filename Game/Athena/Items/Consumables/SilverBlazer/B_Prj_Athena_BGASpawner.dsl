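; Thrown projectile that spawns a configurable Building Gameplay Actor (BGA) where it lands, or reverts to an item pickup.
; On the server, hits pass shared validity checks, while water entry checks water and blocking-volume settings.
; SpawnBGA computes the transform, hides the projectile, then applies weapon wraps, optional marking, a gameplay event and moving-platform attachment.
; A failsafe resolves repeated hits by favoring the higher spawn position.
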(event Collision|EventHit
  (params (MyComp :type Primitive%20Component%20Object%20Reference) (Other :type Actor%20Object%20Reference) (OtherComp :type Primitive%20Component%20Object%20Reference) (bSelfMoved :type Boolean) (HitLocation :type Vector :default "0, 0, 0") (HitNormal :type Vector :default "0, 0, 0") (NormalImpulse :type Vector :default "0, 0, 0") (Hit :type Hit%20Result%20Structure%20%28by%20ref%29))
  ;[0 1 0.1338] Event Hit
  ; Check if the projectile has hit a valid target and determine the behavior of the hit
  (|CallParentFunction MyComp Other OtherComp bSelfMoved HitLocation HitNormal NormalImpulse Hit)
  (bind _inairloop (Variables|Default|GetInAirLoop)
    (:then
      (Audio|Components|Audio|FadeOut _inairloop 0.25 0.0 "Linear" :!class /Script/Engine.AudioComponent))
    (:else
      _))
  (CallFunction|HandleHit Other OtherComp Hit HitLocation))

(event Custom|InitiateSpawnBGA
  (params (HitResult :type Hit%20Result%20Structure) (HitOtherActor :type Actor%20Object%20Reference) (HitOtherComp :type Primitive%20Component%20Object%20Reference) (HitLocation :type Vector))
  (bind (_bblockinghit _binitialoverlap _time _distance _location _impactpoint _normal _impactnormal _physmat _hitactor _hitcomponent) (Collision|BreakHitResult HitResult))
  (bind (_asfort_aipawn _bsuccess) (Utilities|Casting|CastToFortAIPawn HitOtherActor))
  ;[0 0.009347 1] InitiateSpawnBGA
  ; Setup (stop simulating, do once) and then Spawn BGA on valid hit
  (Game|Components|ProjectileMovement|StopSimulating (Variables|Movement|GetProjectileMovementComponent) HitResult)
  (CallFunction|SpawnBGA _normal _hitactor _hitcomponent (Math|Vector|SelectVector HitLocation (CallFunction|GetClosestPoint HitOtherComp (Transformation|GetActorLocation)) (or _bsuccess (Variables|Default|GetUseAbsoluteHitLocation))) false HitResult))

(event Custom|RevertThrownObjectToPickup
  (params (HitResult :type Hit%20Result%20Structure))
  (bind _lifespanafterreverttoobject (Variables|ReverttoObject|GetLifespanAfterRevertToObject))
  ;[0.3021 0.1664 0.007888] RevertThrownObjectToPickup
  (Networking|SwitchHasAuthority
    (:Authority
      (if (or (Variables|Spawning|GetSpawned) (Variables|Default|GetRevertedToPickup))
        (Development|PrintString "can't revert to pickup as it was already spawned/reverted to pickup")
        (else
          (Variables|Default|SetRevertedToPickup true)
          (bind _itemdeftospawnaspickupwheninvalid (Variables|Spawning|GetItemDefToSpawnAsPickupWhenInvalid)
            (:then
              (bind _returnvalue (Inventory|SpawnPickupInWorld _itemdeftospawnaspickupwheninvalid 1 (Transformation|GetActorLocation) "0.000000,0.000000,1.000000" 0 true false false -1 "Tossed" "TossedByPlayer" 0 false))
              (if (Variables|MarkItem|GetMarkItemPickup)
                (CallFunction|Mark _returnvalue)
                (else
                  _)))
            (:else
              ;;[1 0 0] Warning: Item expected to revert to a pickup, but the Item Definition is invalid
              (Fortnite|FortShippingLog self (Utilities|String|Append (Utilities|GetDisplayName self) ": attempt to spawn thrown projectile as pickup failed (invalid ItemDef) - destroying projectile") true)))
          (Rendering|SetActorHiddenInGame (Variables|ReverttoObject|GetSetActorHiddenOnRevertToObject))
          (Collision|SetActorEnableCollision false)
          (Game|Components|ProjectileMovement|StopSimulating (Variables|Movement|GetProjectileMovementComponent) HitResult)
          (if (Variables|RevertToObject|GetDestroyActorImmediatelyOnRevertToObject)
            (Actor|DestroyActor)
            (elif (> _lifespanafterreverttoobject 0)
              (Actor|SetLifeSpan _lifespanafterreverttoobject))))))))

(event OnEnterWaterEvent(WaterInteractionComponent)
  (params (WaterBody :type Fort%20Water%20Body%20Actor%20Object%20Reference) (WaterInteractionComponent :type Fort%20Water%20Interaction%20Component%20Object%20Reference) (bIsFirstBody :type Boolean))
  ;[0 1 0.1338] Spawn on Enter Water
  ; Check water spawning conditions and either spawn or revert the item to a pickup
  (Utilities|FlowControl|Sequence
    (:then_0
      (Networking|SwitchHasAuthority
        (:Authority
          (if (Variables|Spawning|ValidityChecks|GetbSpawnOnCollision)
            (if (Variables|Spawning|ValidityChecks|GetbAllowSpawnOnWater)
              (if (not (CallFunction|CheckIsBlockedbyVolume))
                (CallFunction|SpawnBGA "0, 0, 0" WaterBody 0 (Transformation|GetActorLocation) true))
              (else
                _))
            (CallFunction|RevertThrownObjecttoPickup)))))
    (:then_1
      (if bIsFirstBody
        (CallFunction|HeldWaterImpact)))))

(event Custom|Held%20Water%20Impact
  ;[0.6195 0.07955 1] HeldWaterImpact
  ; SFX and VFX on water impact - play splash sound
  (Audio|PlaySoundatLocation (Variables|Spawning|FX|GetSplash%20Sound) (Transformation|GetActorLocation)))

(event Custom|Mark
  (params (ActorToBeMarked :type Actor%20Object%20Reference))
  ;[1 1 1] Mark. Set Actor to be Marked before calling this. Call from Server only (replicated multicast)
  (Variables|MarkItem|SetActorToBeMarked ActorToBeMarked)
  (bind _asfort_player_pawn_athena (Utilities|Casting|CastToFortPlayerPawnAthena (Actor|GetOwner)))
  (bind _asfort_player_controller_athena (Utilities|Casting|CastToFortPlayerControllerAthena (Pawn|GetController _asfort_player_pawn_athena)))
  (bind _returnvalue (MarkerComponent|MarkActoronServer (Class|FortPlayerController|GetMarkerComponent _asfort_player_controller_athena) (Variables|MarkItem|GetActorToBeMarked) (Variables|MarkItem|GetMarkItem_IncludeSquad) (Variables|MarkItem|GetMarkItem_UseHoveredMarkerDetail)))
  (Variables|MarkItem|SetMarkerID _returnvalue))

(fn CalculateRotation
  (params (NonZeroSurfaceNormal :type Vector))
  (return (Math|Rotator|SelectRotator (Math|Rotator|MakeRotfromXZ (Math|Vector|ProjectVectorontoPlane (Transformation|GetActorForwardVector (Game|GetInstigator)) NonZeroSurfaceNormal) NonZeroSurfaceNormal) (Transformation|GetActorRotation self) (Variables|Spawning|TransformData|GetUseImpactNormalForBGASpawnRotation))))

(fn CheckForWater_
  (params (CurrPos :type Vector) (HitActor :type Actor%20Object%20Reference))
  (bind _asfort_water_body_actor (Utilities|Casting|CastToFortWaterBodyActor HitActor)
    (:then
      (bind _outwatersurfaceinfo (WaterBody|GetWaterSurfaceInfoStruct _asfort_water_body_actor (Transformation|GetActorLocation) true false false))
      (bind _waterplanelocation (Utilities|Struct|BreakWaterSurfaceInfo _outwatersurfaceinfo))
      (bind (_waterplanelocation_x _waterplanelocation_y _waterplanelocation_z) (Math|Vector|BreakVector _waterplanelocation))
      (return (Math|Vector|MakeVector (.x CurrPos) (.y CurrPos) (+ _waterplanelocation_z (Variables|Spawning|TransformData|GetSpawnPosZOffsetInWater)))))
    (:CastFailed
      (return CurrPos))))

(fn CheckIsBlockedByVolume
  ;[0.474 0.6502 1] CheckIsBlockedByVolume
  ; If the object is set to allow blocking volumes to prevent spawning and determine if an overlap is blocking it
  (bind _boverlapping (Default|IsOverlappingDeployableBlockVolume self (Variables|Spawning|ValidityChecks|GetBlockingAreaCheckRadius) (Transformation|GetActorLocation) (Variables|Spawning|GetItem%20Specific%20Blocking%20Tag)))
  (return (select (Variables|Spawning|ValidityChecks|GetbCanSpawnInBlockVolume) (not (Variables|Spawning|ValidityChecks|GetbCanSpawnInBlockVolume)) _boverlapping)))

(fn GetClosestPoint
  (params (Component :type Primitive%20Component%20Object%20Reference) (StartingPoint :type Vector))
  (bind (_outpointonbody _returnvalue) (Collision|GetClosestPointonCollision Component StartingPoint "None"))
  (return (Math|Vector|SelectVector (+ (* (Math|Vector|Normalize (- StartingPoint _outpointonbody)) 18.0) _outpointonbody) (Transformation|GetActorLocation) (> _returnvalue 0.0))))

(fn HandleHit
  (params (OtherActor :type Actor%20Object%20Reference) (OtherComp :type Primitive%20Component%20Object%20Reference) (Hit :type Hit%20Result%20Structure%20%28by%20ref%29) (HitLocation :type Vector))
  (Networking|SwitchHasAuthority
    (:Authority
      (bind (_validspawn _reverttopickup) (CallFunction|IsHitValidforSpawn OtherComp OtherActor Hit))
      (if _validspawn
        (CallFunction|InitiateSpawnBGA Hit OtherActor OtherComp HitLocation)
        (elif _reverttopickup
          (CallFunction|RevertThrownObjecttoPickup Hit))))))

(fn HandleMovingPlatformAttachment
  (params (CollidedObject :type Object%20Reference) (CollidedComponent :type Primitive%20Component%20Object%20Reference))
  (Utilities|FlowControl|Sequence
    (:then_0
      (if (Variables|Spawning|GetbAttachSpawnedObjectOnMovingPlataform)
        (bind _asactor (Utilities|Casting|CastToActor CollidedObject))
        (Utilities|IsValid (Actor|GetComponentbyClass _asactor "/Script/FortniteGame.FortMovingPlatformComponent")
          (:Is%20Valid
            (Utilities|IsValid (Variables|Spawning|GetSpawnedBGA)
              (:Is%20Valid
                (Transformation|AttachActorToComponent (Variables|Spawning|GetSpawnedBGA) CollidedComponent "None" "KeepWorld" "KeepWorld" "KeepWorld" true)
                (bind _asbuilding_gameplay_actor (Utilities|Casting|CastToBuildingGameplayActor (Variables|Spawning|GetSpawnedBGA))
                  (:then
                    ;[1 1 1] This is the same value that pickups start to slide on the train. This is a default value that can be overridden on each BGA blueprint.
                    (Class|BuildingGameplayActor|SetLifeSpanforMovingPlatformAttachedBuildingActors _asbuilding_gameplay_actor (ScalableFloat|GetValueatLevel (Variables|Default|GetLifespanForMovingPlataformAttachedBGAs))))
                  (:CastFailed
                    _))
                (return true)))))))
    (:then_1
      (return false))))

(fn IsHitValidForSpawn
  (params (HitOtherComp :type Primitive%20Component%20Object%20Reference) (HitOther :type Actor%20Object%20Reference) (HitResult :type Hit%20Result%20Structure))
  ;[0.474 0.6502 1] IsHitValidForSpawn
  ; Run through conditions for a valid surface hit and then output the type of action to be taken by the projectile: Invalid (default behavior), Valid (spawns BGA), or Invalid Revert to Pickup (reverts thrown projectile to the pickup version)
  (bind (_validspawn _reverttopickup) (Class|AthenaConsumablesSharedFunctionLibrary|IsHitValidforSpawn HitOtherComp HitOther HitResult (Variables|Spawning|ValidityChecks|GetbSpawnOnCollision) (Variables|Spawning|ValidityChecks|GetDontSpawnOnBGAWithGravity) (Variables|Spawning|ValidityChecks|GetbDontSpawnOnPlayers) (Variables|Spawning|ValidityChecks|GetbDontSpawnOnVehicles) (Variables|Spawning|ValidityChecks|GetbDontSpawnOnUnderneathHit) (Variables|Spawning|ValidityChecks|GetMinDotWithUp) (Variables|Spawning|ValidityChecks|GetbDontSpawnOnPhysicsSim) (Variables|Spawning|ValidityChecks|GetbCanSpawnInBlockVolume) self (Variables|Spawning|ValidityChecks|GetBlockingAreaCheckRadius) (Variables|Spawning|GetItem%20Specific%20Blocking%20Tag)))
  (return _validspawn _reverttopickup))

(fn SpawnBGA_
  (params (Normal :type Vector) (HitActor :type Actor%20Object%20Reference) (HitComp :type Primitive%20Component%20Object%20Reference) (Pos :type Vector) (FromWater :type Boolean) (Hit :type Hit%20Result%20Structure))
  (Utilities|FlowControl|Sequence
    (:then_0
      (Networking|SwitchHasAuthority
        (:Authority
          (if (not (or (Variables|Spawning|GetSpawned) (Variables|Default|GetRevertedToPickup)))
            (Variables|Spawning|SetSpawned true)
            (else
              ;[1 0 0.2117] Failsafe for if two actors are hit on the same frame, use the one with the highest position to avoid clipping
              (bind _spawnedbga (Variables|Spawning|GetSpawnedBGA)
                (:then
                  (bind (_returnvalue_x _returnvalue_y _returnvalue_z) (Math|Vector|BreakVector (Transformation|GetActorLocation _spawnedbga)))
                  (if (> (.z Pos) _returnvalue_z)
                    (Actor|DestroyActor _spawnedbga)
                    (else
                      (Development|PrintString "BGASpawner: Deployed actor destroyed in favor of better hit" true true "(R=1.000000,G=0.861477,B=0.000000,A=1.000000)" 6.0))))
                (:else
                  (Development|PrintString "BGASpawner: Thrown projectile revereted to Pickup" true true "(R=1.000000,G=0.861477,B=0.000000,A=1.000000)" 6.0))) ; should only have a valid BGA if spawned
            ))
          (Utilities|FlowControl|Sequence
            (:then_0
              ;[1 0.6262 0.1371] Cleanup the projectile: set life spawn, disable collision, make hidden, play hit sound
              (Utilities|FlowControl|DoOnce false
                (:Completed
                  (Actor|SetLifeSpan (Variables|Spawning|GetLifespanAfterSpawnBGA))
                  (Collision|SetActorEnableCollision false)
                  (Rendering|SetActorHiddenInGame (Variables|Spawning|GetSetActorHiddenOnSpawnBGA))
                  (Audio|PlaySoundatLocation (Variables|Spawning|FX|GetSoundActivate) (Transformation|GetActorLocation)))) ; should only run on the first hit\ndoes not need to be cleaned up again if\nthe two-hit failsafe is run
            )
            (:then_1
              ;[0.4426 1 0.5749] Spawn the BGA
              (bind _returnvalue (Building|SpawnBuildingActor (Variables|Spawning|GetBGAToSpawn) (Math|Transform|MakeTransform (Math|Vector|vector+vector (CallFunction|CheckforWater Pos HitActor) (* (Utilities|MakeSureNormalIsntZero Normal) (Variables|Spawning|TransformData|GetSpawnPosOffsetAlongNormal)) (Math|Vector|MakeVector 0.0 0.0 (Variables|Spawning|TransformData|GetSpawnPosAdditionalZOffset))) (CallFunction|CalculateRotation (Utilities|MakeSureNormalIsntZero Normal)) (Variables|Spawning|TransformData|GetSpawnScale)) (Variables|Actor|GetInstigator) (Variables|Actor|GetInstigator)))
              (bind _output_get (Variables|Spawning|SetSpawnedBGA _returnvalue))
              (bind _asbuilding_gameplay_actor (Utilities|Casting|CastToBuildingGameplayActor _output_get))
              (Building|InitializeKismetSpawnedBuildingActor _asbuilding_gameplay_actor 0 (select (Variables|Spawning|GetUsePlayerAsSpawningController) (Projectile|GetInstigatorPlayerController) 0) true 0 false)
              (bind _returnvalue_2 (CallFunction|GetOwnerWeapon))
              (Utilities|IsValid _returnvalue_2
                (:Is%20Valid
                  (CosmeticWraps|AttemptApplyOverrideWraptoActor _returnvalue_2 _asbuilding_gameplay_actor))))
            (:then_2
              ;[0.5594 0.7 1] If desired, mark the actor
              (bind _spawnedbga_2 (Variables|Spawning|GetSpawnedBGA)
                (:then
                  (if (and (Variables|MarkItem|GetMarkSpawnedBGA) (Utilities|IsValid _spawnedbga_2 :!class /Script/Engine.KismetSystemLibrary))
                    (CallFunction|Mark _spawnedbga_2)))
                (:else)))
            (:then_3
              ;[0.2215 0.4325 1] Send Event.Projectile.SpawnedBGA event
              (Ability|SendGameplayEventtoActor (Actor|GetOwner) "(TagName=\"Event.Projectile.SpawnedBGA\")" (Utilities|Struct|MakeGameplayEventData :OptionalObject (Variables|Spawning|GetSpawnedBGA) :ContextHandle "()" :InstigatorTags "(GameplayTags=)" :TargetTags "(GameplayTags=)" :EventMagnitude 0.0 :TargetData "()")))
            (:then_4
              ;[0.2215 0.4325 1] Handle Moving Platform Attachment (ex. Train)
              (CallFunction|HandleMovingPlatformAttachment (Variables|GetHitActor) (Variables|GetHitComp)))))) ; Function should only called with authority
    )
    (:then_1
      ;[1 0.4888 0.419] Return
      ; sequence ensures this always executes
      (return (Variables|Spawning|GetSpawnedBGA)))) ; [0] Spawning Logic\n[1] RETURN
)

(fn ConstructionScript)

(fn WillBounceOffTarget
  (params (Hit :type Hit%20Result%20Structure%20%28by%20ref%29))
  (bind (_bblockinghit _binitialoverlap _time _distance _location _impactpoint _normal _impactnormal _physmat _hitactor _hitcomponent) (Collision|BreakHitResult Hit))
  (if (not (Components|ComponentHasTag _hitcomponent "NoStick"))
    (bind _returnvalue (GameplayTags|HasMatchingGameplayTag(Message) _hitactor "(TagName=\"Athena.Item.ShieldBubbleSafeZone\")"))
    (if (not _returnvalue)
      (Utilities|Casting|CastToBuildingGameplayActor _hitactor
        (:then
          (return true))
        (:CastFailed
          _)))
    (return false)))
