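; Projectile Blueprint that spawns a BGA (building gameplay actor) on impact and can attach it to the actor it hit.
; It overrides SpawnBGA, calls the parent to do the spawn, then runs the hit through the Filter variable if the result is valid.
; When the filter passes, the spawned BGA attaches to the hit actor, keeping world transform and welding simulated bodies.
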(fn SpawnBGA_
  (params (Normal :type Vector :default "0, 0, 0") (HitActor :type Actor%20Object%20Reference) (HitComp :type Primitive%20Component%20Object%20Reference) (Pos :type Vector :default "0, 0, 0") (FromWater :type Boolean) (Hit :type Hit%20Result%20Structure))
  (bind (_bblockinghit _binitialoverlap _time _distance _location _impactpoint _normal _impactnormal _physmat _hitactor) (Collision|BreakHitResult Hit))
  (bind _returnvalue (|CallParentFunction Normal HitActor HitComp Pos FromWater Hit))
  (Utilities|IsValid _returnvalue
    (:Is%20Valid
      (bind _batleastoneactorleft (Ability|FilterTargetHitResultList (Variables|Default|GetFilter) 0 (Utilities|Array|MakeArray Hit)))
      (if _batleastoneactorleft
        (Transformation|AttachActorToActor (Variables|Spawning|GetSpawnedBGA) _hitactor "None" "KeepWorld" "KeepWorld" "KeepWorld" true)
        (else
          _))
      (return _returnvalue))
    (:Is%20Not%20Valid
      _)))

(fn ConstructionScript)
