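; Thrown projectile for the SilverBlazer V2 Pocket consumable, which spawns a shield bubble on impact through its parent class.
; It passes through existing shield bubbles by ignoring them on authority hits and restoring pre-bounce velocity a tick later.
; Overlap with a deployable-block volume reverts it to a pickup instead of spawning.
; On hit, the spawned bubble attempts to attach to physics-object actors or tag-matched building gameplay actors.
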(event Collision|EventHit
  (params (MyComp :type Primitive%20Component%20Object%20Reference) (Other :type Actor%20Object%20Reference) (OtherComp :type Primitive%20Component%20Object%20Reference) (bSelfMoved :type Boolean) (HitLocation :type Vector :default "0, 0, 0") (HitNormal :type Vector :default "0, 0, 0") (NormalImpulse :type Vector :default "0, 0, 0") (Hit :type Hit%20Result%20Structure%20%28by%20ref%29))
  ;[0 1 0.9019 0.484] Event Hit()
  (Networking|SwitchHasAuthority
    (:Authority
      (Utilities|IsAShieldBubbleSafeZone Other
        (:Then
          ;;[1 0.4844 0.4844] If we hit a shield, add it to the ignore list
          (Collision|IgnoreActorwhenMoving (Variables|Collision|GetCapsuleComponent) Other true))
        (:Cast%20Failed
          (bind _boverlapping (Default|IsOverlappingDeployableBlockVolume self (Variables|Spawning|ValidityChecks|GetBlockingAreaCheckRadius) (Transformation|GetActorLocation) (Variables|Spawning|GetItem%20Specific%20Blocking%20Tag)))
          (if _boverlapping
            ;;[1 0.4844 0.4844] Turn into Pickup
            (CallFunction|RevertThrownObjecttoPickup Hit)
            (else
              ;;[0 0.1464 0.4531] Spawn and attach the shield bubble
              (|CallParentFunction MyComp Other OtherComp bSelfMoved HitLocation HitNormal NormalImpulse Hit) ; runs additional validity checks and\nspawns the bubble shield (if possible)
              (CallFunction|AttempttoAttachtoHitActor (Variables|Spawning|GetSpawnedBGA) Hit))))))))

(event OnProjectileBounce(ProjectileMovementComponent)
  (params (ImpactResult :type Hit%20Result%20Structure%20%28by%20ref%29) (ImpactVelocity :type Vector%20%28by%20ref%29 :default "0, 0, 0"))
  (bind (_bblockinghit _binitialoverlap _time _distance _location _impactpoint _normal _impactnormal _physmat _hitactor) (Collision|BreakHitResult ImpactResult))
  ;[0.2937 0.562 1 0.524] OnProjectileBounce()
  ; If we hit a shield we want to continue moving through it without bouncing
  ; Immediately add the actor to our ignore list and then set our velocity to the pre-bounce velocity.
  (Utilities|IsAShieldBubbleSafeZone _hitactor
    (:Then
      (Utilities|FlowControl|DelayUntilNextTick) ; We need to wait 1 frame because the Projectile Movement Component is not done updating the velocity for this frame.
      (Projectile|SetVelocity (Variables|Movement|GetProjectileMovementComponent) ImpactVelocity))))

(fn AttemptToAttachToHitActor
  (params (SpawnedBubble :type Actor%20Object%20Reference) (HitResult :type Hit%20Result%20Structure))
  (bind (_bblockinghit _binitialoverlap _time _distance _location _impactpoint _normal _impactnormal _physmat _hitactor) (Collision|BreakHitResult HitResult))
  (Utilities|IsValid SpawnedBubble
    (:Is%20Valid
      (Utilities|FlowControl|Sequence
        (:then_0
          ;[0.3416 0.7515 1] s[0] attempt to attach to actors with a Physics Object Component (vehicles, etc.)
          (Utilities|IsValid (Actor|GetComponentbyClass _hitactor "/Script/FortniteGame.FortPhysicsObjectComponent")
            (:Is%20Valid
              (bind _returnvalue (Transformation|AttachActorToActor SpawnedBubble _hitactor "None" "KeepWorld" "KeepWorld" "KeepWorld" true))
              ;;[0.9237 0.419 1] Return Attachment Success?
              (return _returnvalue))))
        (:then_1
          ;[0.3416 0.7515 1] s[1] also allow attachment to non-physics actors (such as SuperDingo)
          (bind _asbuilding_gameplay_actor (Utilities|Casting|CastToBuildingGameplayActor _hitactor))
          (if (GameplayTags|HasAnyMatchingGameplayTags _asbuilding_gameplay_actor (Variables|Default|GetAttachToActorAdditionalTags))
            (bind _returnvalue_2 (Transformation|AttachActorToActor SpawnedBubble _hitactor "None" "KeepWorld" "KeepWorld" "KeepWorld" true))
            ;;[0.9237 0.419 1] Return Attachment Success?
            (return _returnvalue_2)))
        (:then_2
          _)))
    (:Is%20Not%20Valid
      _))
  ;[0.9237 0.419 1] Return No Valid Attachment
  (return false))

(fn SpawnBGA_
  (params (Normal :type Vector :default "0, 0, 0") (HitActor :type Actor%20Object%20Reference) (HitComp :type Primitive%20Component%20Object%20Reference) (Pos :type Vector :default "0, 0, 0") (FromWater :type Boolean) (Hit :type Hit%20Result%20Structure))
  (bind _boverlapping (Default|IsOverlappingDeployableBlockVolume self (Variables|Spawning|ValidityChecks|GetBlockingAreaCheckRadius) (Transformation|GetActorLocation) (Variables|Spawning|GetItem%20Specific%20Blocking%20Tag)))
  (if _boverlapping
    (CallFunction|RevertThrownObjecttoPickup Hit)
    (else
      (bind _returnvalue (|CallParentFunction Normal HitActor HitComp Pos FromWater Hit))
      (Default|CallEventSpawnedBGA _returnvalue)
      (return _returnvalue))))

(fn ConstructionScript
  (|CallParentFunction))
