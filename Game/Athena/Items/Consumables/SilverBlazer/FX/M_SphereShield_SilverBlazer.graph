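# Additive, unlit, two-sided material for the SilverBlazer consumable's sphere shield effect.
# Emissive layers Fresnel color, scene-intersection glow, a pulsing grid and scrolling lines.
# World-space sphere masks, timed from a GameTime stamp, form a hit wave that distorts the grid and pushes vertices.
# Deploy, DissolveAmount and Shrink parameters drive noisy reveal, opacity breakup and vertex offsets, with quality and shading-path switches simplifying low-end and mobile.
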Material /Game/Athena/Items/Consumables/SilverBlazer/FX/M_SphereShield_SilverBlazer
graph {
 properties {
  bUsedWithMeshDeformer=true
  bUsedWithSkeletalMesh=true
  blendMode=BLEND_Additive
  pixelDepthOffsetMode=PDOM_Legacy
  shadingModel=MSM_Unlit
  twoSided=true
  usedShadingModels=Unlit
 }
 nodes [
  n1 = Divide(A=Time() {period=1}) {constB=7}
  n2 = Add(
   A=MaterialFunctionCall(UVs=TextureCoordinate() {coordinateIndex=1, vTiling=1.5}, Rotation Center=Constant2Vector() {g=0.5}, "Rotation Angle (0-1)"=@n1) {materialFunction=&CustomRotator}.Rotated Values
   B=@n1
  )
  n3 = Multiply(
   A=TextureSample(UVs=@n2) {samplerSource=SSM_Wrap_WorldGroupSettings, texture=/Game/Effects/Fort_Effects/Textures/Grad/gradient_healed}.R
   B=TextureSample(UVs=@n2) {samplerSource=SSM_Wrap_WorldGroupSettings, texture=/Game/Effects/Fort_Effects/Textures/Mask/T_HalfSphere_D}.R
  )
  n4 = Divide(A=Time() {period=1}) {constB=-5}
  n5 = Add(
   A=MaterialFunctionCall(UVs=TextureCoordinate() {coordinateIndex=1, vTiling=1.2}, Rotation Center=Constant2Vector() {g=0.5, r=1}, "Rotation Angle (0-1)"=@n4) {materialFunction=&CustomRotator}.Rotated Values
   B=@n4
  )
  n6 = Multiply(
   A=TextureSample(UVs=@n5) {samplerSource=SSM_Wrap_WorldGroupSettings, texture=&gradient_healed}.R
   B=TextureSample(UVs=@n5) {samplerSource=SSM_Wrap_WorldGroupSettings, texture=&T_HalfSphere_D}.R
  )
  n7 = TextureCoordinate()
  n8 = ScalarParameter() {defaultValue=0.0214, parameterName=LinesScale}
  n9 = Multiply(A=ScalarParameter() {defaultValue=5, group=HitWave, parameterName=Hitwave-DistortAmt3}) {constB=3}
  n10 = Constant()
  n11 = Time()
  n12 = ScalarParameter() {defaultValue=1, desc=Time-stamp, group=HitWave, parameterName=GameTime}
  n13 = Saturate(None=Divide(A=Subtract(A=@n11, B=@n12), B=ScalarParameter() {defaultValue=0.5, group=HitWave, parameterName=HitWave-DurationNew}))
  n14 = Multiply(
   A=ScalarParameter() {defaultValue=300, group=HitWave, parameterName=HitWave-SizeMultiplierOuterRing}
   B=Power(Base=@n13, Exp=ScalarParameter() {defaultValue=0.5, group=HitWave, parameterName=HitWave-OuterRingDelay}) {constExponent=0.7}
  )
  n15 = ScalarParameter() {defaultValue=0.7, group=HitWave, parameterName=Hitwave-SphereHard1}
  n16 = ComponentMask(None=VectorParameter() {group=HitWave, parameterName=HitWave-Loc, defaultValue=<rgba 100 -1400 -150 0>}.RGB) {b=true, g=true, r=true}
  n17 = WorldPosition() {worldPositionShaderOffset=WPT_ExcludeAllShaderOffsets}
  n18 = LinearInterpolate(A=ScalarParameter() {defaultValue=-0.1, group=HitWave, parameterName=HitWave-InnerRingDelay}, Alpha=@n13) {constA=-0.7}
  n19 = SphereMask(
   A=@n17.XYZ
   B=@n16
   Radius=Multiply(A=Saturate(None=@n18), B=ScalarParameter() {defaultValue=1000, group=HitWave, parameterName=HitWave-SizeMultiplierInnerRing})
   Hardness=ScalarParameter() {defaultValue=0.25, group=HitWave, parameterName=Hitwave-SphereHard2}
  ) {hardnessPercent=0}
  n20 = ShadingPathSwitch(Default=Multiply(A=SphereMask(A=@n17.XYZ, B=@n16, Radius=@n14, Hardness=@n15) {hardnessPercent=0}, B=OneMinus(None=@n19)), Mobile=Constant()) {desc=Migrated from FeatureLevelSwitch}
  n21 = QualitySwitch(
   Default=ShadingPathSwitch(Default=@n20, Deferred=@n20, Mobile=@n10) {desc=Migrated from FeatureLevelSwitch}
   Low=@n10
   High=@n20
   Medium=@n10
  )
  n22 = TextureSample(UVs=Add(A=Multiply(A=@n7, B=@n8), B=Multiply(A=@n21, B=@n9))) {samplerType=SAMPLERTYPE_LinearColor, texture=/Game/Athena/SafeZone/Lines}
  n23 = Multiply(
   A=Multiply(A=Add(A=@n3, B=@n6), B=VectorParameter() {group=Grid, parameterName=Grid Color, defaultValue=<rgba 0.0143 0.063 0.15 0>}.RGB)
   B=Multiply(A=@n22.R) {constB=0.7}
  )
  n24 = Constant()
  n25 = ShadingPathSwitch(Default=@n23, Deferred=@n23, Mobile=@n24) {desc=Migrated from FeatureLevelSwitch}
  n26 = VectorParameter() {group=Color, parameterName=ColorCenter, defaultValue=<rgba 0.0096 0.1133 0.1927 1>}
  n27 = Fresnel(
   ExponentIn=ScalarParameter() {defaultValue=3, group=Color, parameterName=FresnelExp}
   BaseReflectFractionIn=ScalarParameter() {defaultValue=0.0952, group=Color, parameterName=FresnelBase}
  )
  n28 = VectorParameter() {group=Color, parameterName=ColorEdge, defaultValue=<rgba 0.0096 0.1133 0.1927 1>}
  n29 = LinearInterpolate(
   A=LinearInterpolate(A=@n26.RGB, B=VectorParameter() {group=Color, parameterName=ColorMid, defaultValue=<rgba 0.0096 0.1133 0.1927 1>}.RGB, Alpha=@n27)
   B=@n28.RGB
   Alpha=Fresnel() {baseReflectFraction=0, exponent=10}
  )
  n30 = LinearInterpolate(A=@n26.RGB, B=Divide(A=@n28.RGB) {constB=4}, Alpha=@n27)
  n31 = Multiply(
   A=ScalarParameter() {defaultValue=1, group=Color, parameterName=FresnelBrightness}
   B=QualitySwitch(
    Default=ShadingPathSwitch(Default=@n29, Deferred=@n29, Mobile=@n30) {desc=Migrated from FeatureLevelSwitch}
    Low=@n30
    High=@n29
    Medium=@n30
   )
  )
  n32 = ScalarParameter() {defaultValue=100, group=IntersectingEdge, parameterName=IntersectingEdgeBias}
  n33 = ScalarParameter() {defaultValue=800, group=IntersectingEdge, parameterName=IntersectingEdgeDivisor}
  n34 = VectorParameter() {group=IntersectingEdge, parameterName=IntersectingEdgeColor, defaultValue=<rgba 0.0096 0.1133 0.1927 1>}
  n35 = Multiply(
   A=Power(
    Base=MaterialFunctionCall!(Distance Bias=@n32, Distance Divisor=@n33, Scene Depth, Pixel Depth) {materialFunction=/Game/Effects/Fort_Effects/Materials/_Functions/MF_DistanceBehindTranslucency_01}.Normalized Inverted
    Exp=ScalarParameter() {defaultValue=4, group=IntersectingEdge, parameterName=Intersection Edge Exponent}
   )
   B=Multiply(A=@n34.RGB, B=ScalarParameter() {defaultValue=1.6779, group=IntersectingEdge, parameterName=IntBrightness})
  )
  n36 = MaterialFunctionCall(Float2=TextureCoordinate()) {materialFunction=&BreakOutFloat2Components}
  n37 = Power(Base=Sine(None=Multiply(A=@n36.G) {constB=0.5}), Exp=ScalarParameter() {defaultValue=5, group=Grid, parameterName=GridFadeTopAndBotton})
  n38 = VectorParameter() {group=HitWave, parameterName=HitWave-Loc, defaultValue=<rgba 100 -1400 -150 0>}
  n39 = ScalarParameter() {defaultValue=300, group=HitWave, parameterName=HitWave-SizeMultiplierOuterRing}
  n40 = TextureCoordinate() {uTiling=5, vTiling=5}
  n41 = Multiply(A=ComponentMask(None=VectorParameter() {group=Grid, parameterName=GridTiling, defaultValue=<rgba 3 3 0 0>}.RGB) {g=true, r=true}) {constB=2}
  n42 = Constant()
  n43 = OneMinus(None=@n13)
  n44 = QualitySwitch(
   Default=ShadingPathSwitch(Default=@n43, Deferred=@n43, Mobile=@n42) {desc=Migrated from FeatureLevelSwitch}
   Low=@n42
   High=@n43
   Medium=@n42
  )
  n45 = Multiply(A=@n21, B=LinearInterpolate(B=ScalarParameter() {defaultValue=5, group=HitWave, parameterName=Hitwave-DistortAmt3}, Alpha=@n44))
  n46 = ScalarParameter() {defaultValue=1, parameterName=HitEffectLerp}
  n47 = TextureSampleParameter2D(UVs=Add(A=Multiply(A=@n40, B=@n41), B=Multiply(A=@n45, B=@n46))) {parameterName=GridTexture, samplerType=SAMPLERTYPE_LinearColor, texture=/Game/Gadgets/Assets/Teleporter/Textures/T_Teleporter_GridSquares_01}
  n48 = Multiply(A=@n47.R) {constB=4}
  n49 = Multiply(A=Saturate(None=Subtract(A=@n47.B) {constB=0.5})) {constB=600, desc=Hit Grid Brightness}
  n50 = ScalarParameter() {defaultValue=1, parameterName=HitEffectLerp}
  n51 = Subtract(A=Time(), B=ScalarParameter() {defaultValue=1, desc=Time-stamp, group=HitWave, parameterName=GameTime})
  n52 = ScalarParameter() {defaultValue=0.5, group=HitWave, parameterName=HitWave-DurationNew}
  n53 = OneMinus(None=Saturate(None=Divide(A=@n51, B=@n52)))
  n54 = Multiply(
   A=SphereMask(
    A=WorldPosition() {worldPositionShaderOffset=WPT_ExcludeAllShaderOffsets}.XYZ
    B=ComponentMask(None=@n38.RGB) {b=true, g=true, r=true}
    Radius=Divide(A=@n39) {constB=2.5}
   ) {attenuationRadius=128, hardnessPercent=0}
   B=Multiply(A=LinearInterpolate(A=@n48, B=@n49, Alpha=@n50), B=Power(Base=@n53) {constExponent=4})
  )
  n55 = Add(A=PixelDepth().R, B=Constant() {r=-500})
  n56 = ScalarParameter() {defaultValue=6500, group=Opacity, parameterName=Distance_Fade}
  n57 = OneMinus(None=Saturate(None=Divide(A=@n55, B=@n56)))
  n58 = Add(
   A=MaterialFunctionCall!(
    TextureObject=TextureObjectParameter() {parameterName=Param, samplerType=SAMPLERTYPE_LinearColor, texture=&HighContrastNoise}
    TextureSize=VectorParameter() {group=Grid, parameterName=CausticWorldScale, defaultValue=<rgba 29400 36750 21000 0>}.RGB
    WorldPosition
    Export Float 4=StaticBool()
    In
    World Space Normal
    ProjectionTransitionContrast=Constant() {r=1}
   ) {materialFunction=/Game/Athena/SafeZone/WorldAlignedTexture_ScalePan}.XY Texture
   B=Multiply(A=Time() {desc=15.015014, period=1}, B=ScalarParameter() {defaultValue=0.08, desc=0.0666, group=Grid, parameterName=Sine Speed A_1})
  )
  n59 = Distance(A=ComponentMask(None=Frac(None=@n58)) {g=true}, B=Constant() {r=0.5})
  n60 = ShadingPathSwitch(Default=OneMinus(None=Multiply(A=@n59) {constB=2}), Mobile=Constant() {r=1}) {desc=Migrated from FeatureLevelSwitch}
  n61 = ScalarParameter() {defaultValue=0.2, group=Grid, parameterName=Ripple to Grid Scale}
  n62 = TextureCoordinate() {uTiling=5, vTiling=5}
  n63 = ComponentMask(None=VectorParameter() {group=Grid, parameterName=GridTiling, defaultValue=<rgba 3 3 0 0>}.RGB) {g=true, r=true}
  n64 = Multiply(A=@n21, B=LinearInterpolate(B=ScalarParameter() {defaultValue=5, group=HitWave, parameterName=Hitwave-DistortAmt3}, Alpha=@n44))
  n65 = ScalarParameter() {defaultValue=1, parameterName=HitEffectLerp}
  n66 = TextureSampleParameter2D(UVs=Add(A=Multiply(A=@n62, B=@n63), B=Multiply(A=@n64, B=@n65))) {parameterName=GridTexture, samplerType=SAMPLERTYPE_LinearColor, texture=&T_Teleporter_GridSquares_01}
  n67 = ComponentMask(None=@n66.RGB) {r=true}
  n68 = Multiply(A=Time() {bOverride_Period=true, period=3.8314}, B=ScalarParameter() {defaultValue=0.261, group=Grid, parameterName=Sine Speed Grid})
  n69 = Sine(None=Add(A=Multiply(A=@n60, B=@n61), B=Add(A=@n67, B=@n68)))
  n70 = ComponentMask(None=@n66.RGB) {g=true}
  n71 = Multiply(A=Power(Base=Multiply(A=@n69, B=@n70), Exp=ScalarParameter() {defaultValue=5, group=Grid, parameterName=Grid Power}), B=@n37)
  n72 = Multiply(A=Add(A=Multiply(A=@n37, B=@n54), B=Multiply(A=@n57, B=@n71)), B=ScalarParameter() {defaultValue=1, group=Grid, parameterName=GridIntensity})
  n73 = VectorParameter() {group=Grid, parameterName=Grid Color, defaultValue=<rgba 0.0143 0.063 0.15 0>}
  n74 = Multiply(A=VectorParameter() {group=HitWave, parameterName=HitWave-Tint, defaultValue=<rgba 0.4034 0.7537 1 0>}.RGB, B=@n20)
  n75 = ScalarParameter() {defaultValue=1, parameterName=HitEffectLerp}
  n76 = LinearInterpolate(B=Multiply(A=Multiply(A=@n47.R, B=@n74) {constB=15}, B=LinearInterpolate(Alpha=@n75) {constB=85}) {constB=85}, Alpha=@n43)
  n77 = Constant()
  n78 = QualitySwitch(
   Default=ShadingPathSwitch(Default=@n76, Deferred=@n76, Mobile=@n77) {desc=Migrated from FeatureLevelSwitch}
   Low=@n77
   High=@n76
   Medium=@n77
  )
  n79 = Power(Base=@n43) {constExponent=4}
  n80 = VectorParameter() {group=HitWave, parameterName=Hitwave-AddBrighness, defaultValue=<rgba 0.0173 0.1326 0.2031 0>}
  n81 = Add(
   A=Add(A=Add(A=@n31, B=@n35) {desc=+ intersection edge}, B=Multiply(A=@n72, B=@n73.RGB))
   B=Add(A=Add(A=@n78, B=@n74), B=Multiply(A=@n79, B=@n80.RGB))
  )
  n82 = Constant()
  n83 = MaterialFunctionCall(Float2=TextureCoordinate()) {materialFunction=&BreakOutFloat2Components}
  n84 = OneMinus(None=Sine(None=Multiply(A=@n83.G) {constB=0.5}))
  n85 = TextureSample(UVs=Multiply(A=TextureCoordinate()) {constB=2}) {samplerType=SAMPLERTYPE_Grayscale, texture=&LowResBlurredNoise}
  n86 = MaterialFunctionCall() {materialFunction=&DebugTimeSine}
  n87 = ScalarParameter() {parameterName=Deploy}
  n88 = Multiply(A=LinearInterpolate(Alpha=@n85.R) {constA=0.7}, B=OneMinus(None=StaticSwitchParameter(True=@n86.Result, False=@n87) {parameterName=Debug}))
  n89 = Saturate(None=Subtract(A=@n84, B=LinearInterpolate(Alpha=@n88) {constA=-0.2}))
  n90 = VectorParameter() {group=IntersectingEdge, parameterName=IntersectingEdgeColor, defaultValue=<rgba 0.0096 0.1133 0.1927 1>}
  n91 = VectorParameter() {group=Grid, parameterName=GridTiling, defaultValue=<rgba 3 3 0 0>}
  n92 = TextureSample(UVs=Multiply(A=TextureCoordinate() {uTiling=5, vTiling=5}, B=ComponentMask(None=@n91.RGB) {g=true, r=true})) {samplerType=SAMPLERTYPE_LinearColor, texture=&T_Teleporter_GridSquares_01}
  n93 = Add(A=Multiply(A=@n89) {constB=30}, B=LinearInterpolate(Alpha=Multiply(A=@n92.R, B=@n92.R)) {constA=-1, constB=3}) {constB=0}
  n94 = Multiply(A=MaterialFunctionCall(Float2=TextureCoordinate()) {materialFunction=&BreakOutFloat2Components}.G) {constB=0.5}
  n95 = Multiply(A=Multiply(A=@n89, B=Multiply(A=@n90.RGB, B=@n93)), B=Power(Base=Sine(None=@n94)) {constExponent=4})
  n96 = QualitySwitch(
   Default=ShadingPathSwitch(Default=@n95, Deferred=@n95, Mobile=@n82) {desc=Migrated from FeatureLevelSwitch}
   Low=@n82
   High=@n95
   Medium=@n82
  )
  MaterialFunctionCall(Emissive=Add(A=QualitySwitch(Default=@n25, Low=@n24, High=@n23, Medium=@n24), B=Add(A=@n81, B=@n96))) {materialFunction=&MF_FX_EmissiveAdjustments}
  n97 = Constant()
  n98 = QualitySwitch(
   Default=ShadingPathSwitch(Default=@n22.R, Deferred=@n22.R, Mobile=@n97) {desc=Migrated from FeatureLevelSwitch}
   Low=@n97
   High=@n22.R
   Medium=@n97
  )
  n99 = PixelDepth()
  n100 = Constant() {r=-100}
  n101 = Saturate(None=Divide(A=Add(A=@n99.R, B=@n100), B=ScalarParameter() {defaultValue=300, parameterName=NearFade}))
  n102 = Multiply(A=Add(A=@n35), B=@n101)
  n103 = ScalarParameter() {defaultValue=0.4238, group=Opacity, parameterName=OpacityInside}
  n104 = ScalarParameter() {defaultValue=0.5762, group=Opacity, parameterName=OpacityOutside}
  n105 = Saturate(None=TwoSidedSign())
  n106 = Multiply(A=Add(A=@n35, B=LinearInterpolate(A=@n103, B=@n104, Alpha=@n105)), B=@n101)
  n107 = ComponentMask(None=VectorParameter() {group=Grid, parameterName=GridTiling, defaultValue=<rgba 3 3 0 0>}.RGB) {g=true, r=true}
  n108 = TextureSampleParameter2D(UVs=Multiply(A=TextureCoordinate(), B=Multiply(A=@n107) {constB=6})) {parameterName=Disolve_Testure, samplerType=SAMPLERTYPE_Masks, texture=/Game/Athena/Playlists/Barrier/Textures/T_BarrierGrid_M}
  n109 = Time()
  n110 = Frac(None=Add(A=@n108.R, B=Frac(None=@n109)))
  n111 = OneMinus(None=Multiply(A=Sine(None=@n110), B=@n108.G))
  n112 = ScalarParameter() {group=Opacity, parameterName=DissolveAmount}
  n113 = Multiply(A=ShadingPathSwitch(Default=@n111, Deferred=@n111, Forward=@n108.R, Mobile=@n108.R), B=@n112)
  n114 = Multiply(A=OneMinus(None=Saturate(None=@n113)), B=OneMinus(None=@n112))
  n115 = Constant() {r=1}
  n116 = ShadingPathSwitch(Default=@n114, Deferred=@n114, Mobile=@n115) {desc=Migrated from FeatureLevelSwitch}
  n117 = TextureCoordinate()
  n118 = Sine(None=Multiply(A=MaterialFunctionCall(Float2=@n117) {materialFunction=&BreakOutFloat2Components}.G) {constB=0.5})
  n119 = LinearInterpolate(Alpha=StaticSwitchParameter(
    True=MaterialFunctionCall() {materialFunction=&DebugTimeSine}.Result
    False=ScalarParameter() {parameterName=Deploy}
   ) {parameterName=Debug}) {constA=-1, constB=0.2}
  n120 = LinearInterpolate(Alpha=@n85.R) {constA=0.8}
  n121 = Constant() {r=0.8}
  n122 = ShadingPathSwitch(Default=@n120, Deferred=@n120, Mobile=@n121) {desc=Migrated from FeatureLevelSwitch}
  n123 = Constant()
  n124 = Multiply(
   A=Multiply(
    A=ShadingPathSwitch(Default=@n102, Deferred=@n106, Mobile=@n102) {desc=Migrated from FeatureLevelSwitch}
    B=QualitySwitch(Default=@n116, Low=@n115, High=@n114, Medium=@n115)
   )
   B=If(
    A=Subtract(A=@n118, B=@n119)
    B=QualitySwitch(Default=@n122, Low=@n121, High=@n120, Medium=@n121)
    "A > B"=Constant() {r=1}
    "A == B"=@n123
    "A < B"=@n123
   )
  )
  n125 = DepthFade(Opacity=@n124, FadeDistance=ScalarParameter() {parameterName=Depthfade})
  Multiply(
   A=Multiply(
    A=LinearInterpolate(Alpha=@n98) {constA=0.2, constB=1.5}
    B=ShadingPathSwitch(Default=@n125, Mobile=@n124) {desc=Migrated from FeatureLevelSwitch}
   )
   B=ScalarParameter() {defaultValue=1, parameterName=OverallOpacity}
  )
  n126 = Panner(Coordinate=TextureCoordinate() {vTiling=0.8}) {speedY=0.4}
  n127 = Multiply(A=VertexNormalWS(), B=Multiply(A=TextureSample(UVs=@n126) {texture=&gradient_healed}.R) {constB=12})
  n128 = Constant()
  n129 = QualitySwitch(
   Default=ShadingPathSwitch(Default=@n127, Deferred=@n127, Mobile=@n128) {desc=Migrated from FeatureLevelSwitch}
   Low=@n128
   High=@n127
   Medium=@n128
  )
  n130 = Saturate(None=Power(Base=Saturate(None=@n20), Exp=ScalarParameter() {defaultValue=5, group=HitWave, parameterName=Hitwave-pow}))
  n131 = ScalarParameter() {defaultValue=16, group=HitWave, parameterName=WPO-AddWave2}
  n132 = VertexNormalWS()
  n133 = Constant()
  n134 = LinearInterpolate(Alpha=@n88) {constA=-1, constB=2}
  n135 = LinearInterpolate(Alpha=Saturate(None=Subtract(A=@n84, B=@n134))) {constB=70}
  n136 = ShadingPathSwitch(Default=@n135, Deferred=@n135, Mobile=@n133) {desc=Migrated from FeatureLevelSwitch}
  n137 = Add(A=Multiply(A=Multiply(A=@n130, B=@n131), B=@n132), B=Multiply(A=@n132, B=QualitySwitch(Default=@n136, Low=@n133, High=@n135, Medium=@n133)))
  n138 = VertexNormalWS()
  n139 = TextureSample() {samplerType=SAMPLERTYPE_Grayscale, texture=&LowResBlurredNoise}
  n140 = Constant() {r=1}
  n141 = ScalarParameter() {parameterName=Shrink}
  n142 = Multiply(
   A=QualitySwitch(
    Default=ShadingPathSwitch(Default=@n139.R, Deferred=@n139.R, Mobile=@n140) {desc=Migrated from FeatureLevelSwitch}
    Low=@n140
    High=@n139.R
    Medium=@n140
   )
   B=@n141
  )
  n143 = LinearInterpolate(A=Multiply(A=Multiply(A=@n138) {constB=-1}, B=LinearInterpolate(Alpha=@n142) {constB=3000}), Alpha=@n141) {constB=0}
  n144 = Multiply(A=Multiply(A=VertexNormalWS()) {constB=-1}, B=LinearInterpolate(Alpha=@n141) {constB=800})
  Reroute(None=Add(A=Add(A=@n137, B=@n129), B=Add(A=@n143, B=@n144)) {constB=0})
 ]
}
