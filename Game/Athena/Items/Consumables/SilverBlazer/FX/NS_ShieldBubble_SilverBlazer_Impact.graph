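# Niagara impact effect for the SilverBlazer consumable's shield bubble, built from instantaneous particle bursts.
# Burn_Hole sprites cool from white-hot HDR to fading blue, alongside a brief warm light flash.
# Squares throws randomized cyan sprites along X, slowed by drag as they shrink.
# User parameters set hole count and spread plus square range and size; Ring, Ripple, Impact_Star_Mesh and Sparks_Small carry no scripts or renderers here.
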NiagaraSystem /Game/Athena/Items/Consumables/SilverBlazer/FX/NS_ShieldBubble_SilverBlazer_Impact
dependencies {
 usedDataInterfaces [&NiagaraDataInterfaceColorCurve, &NiagaraDataInterfaceCurve]
 usedDynamicInputs []
 usedModules [&SystemState]
}
emitters {
 Burn_Hole {
  emitterUpdateScript [
   &EmitterState()
   &SpawnBurst_Instantaneous(Spawn Count=$User.ImpactHoleAmount, Use Spawn Probability=false, UseLoopCountLimit=false)
  ]
  particleSpawnScript [
   &InitializeParticle(Lifetime=0.7, Sprite Size Mode=::NewEnumerator3, UsePositionOffset=false, Write Material Random=true)
   &ShapeLocation(
    MaskSpawnGroup=false
    Rotation Mode=::NewEnumerator1
    Sphere Distribution=::NewEnumerator1
    Sphere Radius=$User.ImpactHoleRange
    Sphere Radius Position:NiagaraFloat=&RandomRangeFloat(Minimum=0, Maximum=1)
    U Position=&RandomRangeFloat(Minimum=0, Maximum=1)
    UseOwnerScale=true
   )
   &CameraOffset(Camera Offset Amount=2)
   &SpriteFacingAndAlignment(WriteAlignment=false, WriteFacing=true)
  ]
  particleUpdateScript [
   &ParticleState()
   &ScaleSpriteSize(Uniform Curve Sprite Scale=DataInterface() {propertyValues {Curve {keys [{tangentMode=RCTM_User, time=0.5, value=1}, {tangentMode=RCTM_User, time=1, value=0.4}]}}})
   &ScaleColor(
    Linear Color Curve=DataInterface() {propertyValues {
      AlphaCurve {keys [<at 0 1>,<at 1 0>]}
      BlueCurve {keys [<at 0 67.751>,<at 0.2 76.731>,<at 1 1>]}
      GreenCurve {keys [<at 0 83.7384>,<at 0.2 31.8742>,<at 1 0.8393>]}
      RedCurve {keys [<at 0 100>,<at 0.2 4.4489>,<at 1 0.2926>]}
     }}
    Scale Mode=::NewEnumerator2
   )
  ]
  rendererClasses [&NiagaraSpriteRendererProperties]
  renderers [{rendererClass=&NiagaraSpriteRendererProperties}]
 }
 Impact_Star_Mesh {rendererClasses [], renderers []}
 Light {
  emitterUpdateScript [
   &EmitterState()
   &SpawnBurst_Instantaneous(Use Spawn Probability=false, UseLoopCountLimit=false)
  ]
  particleSpawnScript [&InitializeParticle(Color Mode=::NewEnumerator0, Lifetime=0.3, UsePositionOffset=false, Write Material Random=true)]
  particleUpdateScript [
   &ParticleState()
   &ScaleColor(
    Linear Color Curve=DataInterface() {propertyValues {
      AlphaCurve {keys [<at 0 5>,<at 1 0>]}
      BlueCurve {keys [<at 0 0.4778>,<at 1 0.3442>]}
      GreenCurve {keys [<at 0 1.9023>,<at 1 0.9347>]}
      RedCurve {keys [<at 0 3>,<at 1 2>]}
     }}
    Scale Mode=::NewEnumerator2
   )
  ]
  rendererClasses [&NiagaraLightRendererProperties]
  renderers [{rendererClass=&NiagaraLightRendererProperties, rendererIndex=0}]
 }
 Ring {rendererClasses [], renderers []}
 Ripple {rendererClasses [], renderers []}
 Sparks_Small {rendererClasses [], renderers []}
 Squares {
  emitterUpdateScript [
   &EmitterState(Life Cycle Mode=::NewEnumerator1, Loop Behavior=::NewEnumerator1, Loop Duration=0.3, UseLoopDelay=false)
   &SpawnBurst_Instantaneous(
    Spawn Count=&RandomRangeInt(Minimum=18, Maximum=25)
    Use Spawn Probability=false
    UseLoopCountLimit=false
   )
  ]
  particleSpawnScript [
   &InitializeParticle(
    Color=<rgba 0.0991 1.1317 2 1>
    Lifetime Max=0.35
    Lifetime Min=0.15
    Lifetime Mode=::NewEnumerator1
    Sprite Size Mode=::NewEnumerator1
    Uniform Sprite Size Max=22
    UsePositionOffset=false
    Write Material Random=true
   )
   &ShapeLocation(MaskSpawnGroup=false, Override Seed=false, Sphere Radius=$User.GridRange, Sphere Surface Distribution=1, UseOwnerScale=true)
   &AddVelocity(Velocity=&RandomRangeVector(Minimum=<50 0 0>, Maximum=<800 0 0>))
  ]
  particleUpdateScript [
   &Drag(Drag=&RandomRangeFloat(Minimum=1, Maximum=6), Use Linear Drag=true, Use Rotational Drag=true)
   &SolveForcesAndVelocity(Clamp Velocity=false, Limit Acceleration=false)
   &ParticleState()
   &ScaleSpriteSize(
    Uniform Curve Scale=$User.GridSize
    Uniform Curve Sprite Scale=DataInterface() {propertyValues {Curve {keys [{tangentMode=RCTM_User, time=0, value=2}, <at 0.2 1>, {tangentMode=RCTM_User, time=1, value=0}]}}}
   )
  ]
  rendererClasses [&NiagaraSpriteRendererProperties]
  renderers [{rendererClass=&NiagaraSpriteRendererProperties}]
 }
}
summary {
 systemName=NS_ShieldBubble_SilverBlazer_Impact
 emitters [
  {emitterName=Ring, rendererClasses []}
  {emitterName=Impact_Star_Mesh, rendererClasses []}
  {emitterName=Ripple, rendererClasses []}
  {emitterName=Sparks_Small, rendererClasses []}
  {emitterName=Light, rendererClasses [&NiagaraLightRendererProperties]}
  {emitterName=Burn_Hole, rendererClasses [&NiagaraSpriteRendererProperties]}
  {emitterName=Squares, rendererClasses [&NiagaraSpriteRendererProperties]}
 ]
 userVariables [
  User.GridRange {type=NiagaraFloat, defaultValue=50, description=""}
  User.GridSize {type=NiagaraFloat, defaultValue=1, description=""}
  User.ImpactHoleAmount {type=NiagaraInt32, defaultValue=1}
  User.ImpactHoleRange {type=NiagaraFloat, defaultValue=0, description=""}
 ]
}
