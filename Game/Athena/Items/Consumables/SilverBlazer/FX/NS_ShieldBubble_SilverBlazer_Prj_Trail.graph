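# Niagara projectile trail effect for the SilverBlazer shield bubble consumable, pairing a glowing blue core with a smoke ribbon.
# Prj_Core_and_Light bursts a persistent camera-offset sprite with a blue light, pulsing its size through a waveform-indexed curve.
# Smoke_Ribbon spawns continuously with random drift and upward acceleration, widening as its color curve cools from hot cyan-blue to dim grey.
# Both emitters drive dynamic material parameters, with the ribbon animating them over particle life.
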NiagaraSystem /Game/Athena/Items/Consumables/SilverBlazer/FX/NS_ShieldBubble_SilverBlazer_Prj_Trail
dependencies {
 usedDataInterfaces [&NiagaraDataInterfaceCurve, &NiagaraDataInterfaceColorCurve]
 usedDynamicInputs []
 usedModules [&SystemState]
}
emitters {
 Prj_Core_and_Light {
  emitterUpdateScript [
   &EmitterState(Inactive Response=::NewEnumerator1, Life Cycle Mode=::NewEnumerator1, Loop Duration=1, UseLoopDelay=false)
   &SpawnBurst_Instantaneous(Use Spawn Probability=false, UseLoopCountLimit=true)
  ]
  particleSpawnScript [
   &InitializeParticle(
    Color=<rgba 0.1314 1.0391 2 1>
    Lifetime=5
    Sprite Size=<100 100>
    Sprite Size Mode=::NewEnumerator4
    UsePositionOffset=false
    Write Material Random=true
   )
   &CameraOffset(Camera Offset Amount=50)
   &DynamicMaterialParameters(
    Index 0 Param 1=0.2
    Index 0 Param 2=1
    Index 0 Param 3=1
    Module.Index0Param0Write=true
    Module.Index0Param1Write=true
    Module.Index0Param2Write=true
    Module.Index0Param3Write=false
   )
  ]
  particleUpdateScript [
   &ParticleState(Kill Particles When Lifetime Has Elapsed=false, Let Infinitely Lived Particles Die When Emitter Deactivates=true)
   &ScaleSpriteSize(
    Uniform Curve Index=&Waveform("Waveform [1]"=::NewEnumerator0, "[1] Frequency"=4, "[1] Amplitude Min/Max"=<0 1>)
    Uniform Curve Sprite Scale=DataInterface() {propertyValues {Curve {keys [{tangentMode=RCTM_User, time=0, value=1}, {tangentMode=RCTM_User, time=1, value=0.5}]}}}
   )
   /SetVariables(Particles.LightColor=<rgba 0 0.1516 1 5>) {isSetParameters=true}
  ]
  rendererClasses [&NiagaraSpriteRendererProperties, &NiagaraLightRendererProperties]
  renderers [
   {rendererClass=&NiagaraSpriteRendererProperties}
   {rendererClass=&NiagaraLightRendererProperties, rendererIndex=1}
  ]
 }
 Smoke_Ribbon {
  emitterUpdateScript [
   &EmitterState()
   &SpawnRate(SpawnRate=15, Use Spawn Probability=false)
  ]
  particleSpawnScript [
   &InitializeParticle(
    Color Maximum=<rgba 0.3 0.3 0.3 0.5>
    Color Minimum=<rgba 0.07 0.07 0.07 0.5>
    Color Mode=::NewEnumerator2
    Lifetime=2
    Ribbon Width=15
    Ribbon Width Mode=::NewEnumerator1
    UsePositionOffset=false
    Write Material Random=true
   )
   &AddVelocity(Velocity=&RandomRangeVector(Minimum=<-100 -100 -100>, Maximum=<100 100 100>))
  ]
  particleUpdateScript [
   &ParticleState()
   &AccelerationForce(Acceleration=<0 0 50>)
   &SolveForcesAndVelocity(Clamp Velocity=false, Limit Acceleration=false)
   &DynamicMaterialParameters(
    Index 0 Param 1=&FloatFromCurve(FloatCurve=DataInterface() {propertyValues {Curve {keys [<at 0 1>,<at 1 0>]}}})
    Index 0 Param 3=-0.2
    Index 0 Param 4=&FloatFromCurve(FloatCurve=DataInterface() {propertyValues {Curve {keys [<at 0.5 1 RCIM_Cubic>, {arriveTangent=-3.8291, interpMode=RCIM_Cubic, leaveTangent=-3.8291, tangentMode=RCTM_User, time=1, value=0}]}}})
    Module.Index0Param0Write=true
    Module.Index0Param1Write=true
    Module.Index0Param2Write=true
    Module.Index0Param3Write=true
   )
   &ScaleRibbonWidth(Ribbon Width Scale=&FloatFromCurve(FloatCurve=DataInterface() {propertyValues {Curve {keys [<at 0 0.95>,<at 0.5 9.85>,<at 1 11.4>]}}}))
   &ScaleColor(
    Scale Alpha=&FloatFromCurve(FloatCurve=DataInterface() {propertyValues {Curve {keys [<at 0 1.15>,<at 0.4 0.9734>,<at 0.9272 0.8698>,<at 1 0.8251>]}}})
    Scale RGB=&MakeVectorFromLinearColorRGB(Color=&ColorFromCurve(ColorCurve=DataInterface() {propertyValues {
        AlphaCurve {keys []}
        BlueCurve {keys [<at 0 20 RCIM_Cubic>, {arriveTangent=-39.625, leaveTangent=-39.625, time=0.12, value=0.3984}, <at 0.5 0.1875 RCIM_Cubic>]}
        GreenCurve {keys [<at 0 12.132 RCIM_Cubic>, {arriveTangent=-23.8889, leaveTangent=-23.8889, time=0.12, value=0.4287}, <at 0.5 0.1875 RCIM_Cubic>]}
        RedCurve {keys [<at 0 0 RCIM_Cubic>, {arriveTangent=0.4, leaveTangent=0.4, time=0.12, value=0.5}, <at 0.5 0.2 RCIM_Cubic>]}
       }}))
    ScaleA=true
    ScaleRGB=true
   )
  ]
  rendererClasses [&NiagaraRibbonRendererProperties]
  renderers [{rendererClass=&NiagaraRibbonRendererProperties}]
 }
}
summary {
 systemName=NS_ShieldBubble_SilverBlazer_Prj_Trail
 emitters [
  {emitterName=Smoke_Ribbon, rendererClasses [&NiagaraRibbonRendererProperties]}
  {
   emitterName=Prj_Core_and_Light
   rendererClasses [&NiagaraSpriteRendererProperties, &NiagaraLightRendererProperties]
  }
 ]
 userVariables []
}
