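# Niagara effect for the SilverBlazer shield bubble consumable, which by its name plays when the bubble's time runs out.
# It holds three emitters, CoreFlare, Mesh_Castoff and Squares, named for a central flare, cast-off mesh pieces and square particles.
# SystemState is the only module shown, and no renderers, data interfaces or user parameters are listed.
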NiagaraSystem /Game/Athena/Items/Consumables/SilverBlazer/FX/NS_ShieldBubble_SliverBlazer_TimesUp
dependencies {usedDataInterfaces [], usedDynamicInputs [], usedModules [&SystemState]}
emitters {CoreFlare {rendererClasses [], renderers []}, Mesh_Castoff {rendererClasses [], renderers []}, Squares {rendererClasses [], renderers []}}
summary {
 systemName=NS_ShieldBubble_SliverBlazer_TimesUp
 emitters [{emitterName=CoreFlare, rendererClasses []}, {emitterName=Mesh_Castoff, rendererClasses []}, {emitterName=Squares, rendererClasses []}]
 userVariables []
}
